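# Control Rig for the F_MED_IvyStub female body that correctively moves skirt and collar base bones as the thighs and head move.
# Spherical pose readers store thigh forward, side, back and twist, plus head-back, as alpha variables.
# Smoothed alphas weight bone offsets, while the front skirt averages both thighs, then biases toward the more forward one via parent constraints.
# Some readings, including both SideIn alphas, go unused.
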ControlRigBlueprint /BRCosmetics/Characters/Player/Female/Medium/Bodies/F_MED_IvyStub/Meshes/F_MED_IvyStub_CtrlRig
graphs [
 RigVMFunctionLibrary {links [], nodes []}
 RigVMFunction_Sequence_1_ContainedGraph {nodes [
   n1 = RigVMFunctionEntryNode()
   RigVMFunction_Sequence = &RigVMFunction_Sequence(ExecuteContext=@n1.ExecuteContext)
   n2 = &RigVMFunction_Sequence(ExecuteContext=@RigVMFunction_Sequence.B)
   n3 = &RigVMFunction_Sequence(ExecuteContext=@n2.B)
   n4 = &RigVMFunction_Sequence(ExecuteContext=@n3.B)
   n5 = &RigVMFunction_Sequence(ExecuteContext=@n4.B)
   n6 = &RigVMFunction_Sequence(ExecuteContext=@n5.B)
   RigVMFunctionReturnNode(A=@RigVMFunction_Sequence.A, B=@n2.A, C=@n3.A, D=@n4.A, E=@n5.A, F=@n6.A, G=@n6.B)
  ]}
 RigVMModel {nodes [
   RigUnit_BeginExecution = &RigUnit_BeginExecution()
   RigVMFunction_Sequence = RigVMAggregateNode(ExecuteContext=@RigUnit_BeginExecution.ExecutePin)
   &RigUnit_ParentConstraint(
    AdvancedSettings=(InterpolationType=Average,RotationOrderForFilter=XZY)
    Child=(Type=Bone,Name="dyn_skirt_fr_c_base")
    Parents=((Item=(Type=Bone,Name="thigh_l"),Weight=0.5),(Item=(Type=Bone,Name="thigh_r"),Weight=0.5))
    Weight=1
    ExecutePin=@RigVMFunction_Sequence.C
   )
   n1 = RigVMVariableNode() {variable=L thighForward Alpha}
   n2 = RigVMVariableNode() {variable=R thighForward Alpha}
   Subtract = &RigVMFunction_MathDoubleSub(A=@n1.Value, B=@n2.Value)
   n3 = &RigVMFunction_AlphaInterp(InterpSpeedDecreasing=30, InterpSpeedIncreasing=30, Scale=1, bClampResult=true, Value=@Subtract.Result)
   > &RigUnit_ParentConstraint(
    AdvancedSettings=(InterpolationType=Average,RotationOrderForFilter=XZY)
    Child=(Type=Bone,Name="dyn_skirt_fr_c_base")
    Parents=((Item=(Type=Bone,Name="thigh_l"),Weight=1))
    Weight=@n3.Result
   )
   > &RigUnit_ModifyTransforms(
    ItemToModify=((Item=(Type=Bone,Name="dyn_skirt_fr_c_base"),Transform=(Rotation=(X=-0.342,Y=0,Z=0,W=0.9397),Translation=(X=0,Y=0,Z=0),Scale3D=(X=1,Y=1,Z=1))))
    WeightMinimum=0
    Weight=@n3.Result
   )
   n4 = &RigVMFunction_MathDoubleSub(B=@n1.Value, A=@n2.Value)
   n5 = &RigVMFunction_AlphaInterp(InterpSpeedDecreasing=30, InterpSpeedIncreasing=30, Scale=1, bClampResult=true, Value=@n4.Result)
   > &RigUnit_ParentConstraint(
    AdvancedSettings=(InterpolationType=Average,RotationOrderForFilter=XZY)
    Child=(Type=Bone,Name="dyn_skirt_fr_c_base")
    Parents=((Item=(Type=Bone,Name="thigh_r"),Weight=1))
    Weight=@n5.Result
   )
   > &RigUnit_ModifyTransforms(
    ItemToModify=((Item=(Type=Bone,Name="dyn_skirt_fr_c_base"),Transform=(Rotation=(X=0.342,Y=0,Z=0,W=0.9397),Translation=(X=0,Y=0,Z=0),Scale3D=(X=1,Y=1,Z=1))))
    WeightMinimum=0
    Weight=@n5.Result
   )
   n6 = &RigVMFunction_AlphaInterp(
    InterpSpeedDecreasing=30
    InterpSpeedIncreasing=30
    Scale=1
    bClampResult=true
    Value=&RigVMFunction_MathDoubleMul(B=RigVMVariableNode() {variable=R thighForward Alpha}.Value, A=RigVMVariableNode() {variable=L thighForward Alpha}.Value).Result
   )
   > &RigUnit_ModifyTransforms(
    ItemToModify=((Item=(Type=Bone,Name="dyn_skirt_fr_c_base"),Transform=(Rotation=(X=-0.0858,Y=0,Z=0.0151,W=0.9962),Translation=(X=0.4313,Y=0,Z=2.4461),Scale3D=(X=1,Y=1,Z=1))))
    WeightMinimum=0
    Weight=@n6.Result
   )
   n7 = &RigUnit_SphericalPoseReader(
    ActiveRegionScaleFactors=(PositiveWidth=4,NegativeWidth=3,PositiveHeight=2.5,NegativeHeight=2)
    ActiveRegionSize=0.1
    Debug=(bDrawDebug=false,bDraw2D=False,bDrawLocalAxes=False,DebugScale=25,DebugSegments=20,DebugThickness=0.25)
    DriverAxis=(X=-1,Y=0,Z=0)
    DriverItem=(Type=Bone,Name="thigh_l")
    FalloffRegionScaleFactors=(PositiveWidth=1,NegativeWidth=3,PositiveHeight=1,NegativeHeight=0.5)
    FalloffSize=0.4
    FlipHeightScaling=False
    FlipWidthScaling=False
    OptionalParentItem=(Type=Bone,Name="None")
    RotationOffset=(X=100,Y=0,Z=-20)
    ExecutePin=@RigVMFunction_Sequence.A
   )
   n8 = RigVMVariableNode(Value=@n7.OutputParam, ExecuteContext=@RigVMFunction_Sequence.B) {variable=L thighForward Alpha}
   n9 = &RigUnit_SphericalPoseReader(
    ActiveRegionScaleFactors=(PositiveWidth=4,NegativeWidth=3,PositiveHeight=2.5,NegativeHeight=2)
    ActiveRegionSize=0.1
    Debug=(bDrawDebug=false,bDraw2D=False,bDrawLocalAxes=False,DebugScale=25,DebugSegments=20,DebugThickness=0.25)
    DriverAxis=(X=1,Y=0,Z=0)
    DriverItem=(Type=Bone,Name="thigh_r")
    FalloffRegionScaleFactors=(PositiveWidth=1,NegativeWidth=3.6,PositiveHeight=1,NegativeHeight=0.5)
    FalloffSize=0.4
    FlipHeightScaling=False
    FlipWidthScaling=true
    OptionalParentItem=(Type=Bone,Name="None")
    RotationOffset=(X=-80,Y=0,Z=-35.3405)
    ExecutePin=@n7.ExecutePin
   )
   n10 = RigVMVariableNode(Value=@n9.OutputParam, ExecuteContext=@n8.ExecuteContext) {variable=R thighForward Alpha}
   n11 = &RigUnit_SphericalPoseReader(
    ActiveRegionScaleFactors=(PositiveWidth=3,NegativeWidth=2,PositiveHeight=2,NegativeHeight=2)
    ActiveRegionSize=0.1
    Debug=(bDrawDebug=false,bDraw2D=False,bDrawLocalAxes=False,DebugScale=25,DebugSegments=20,DebugThickness=0.25)
    DriverAxis=(X=-1,Y=0,Z=0)
    DriverItem=(Type=Bone,Name="thigh_l")
    FalloffRegionScaleFactors=(PositiveWidth=2,NegativeWidth=2,PositiveHeight=2,NegativeHeight=2)
    FalloffSize=0.25
    FlipHeightScaling=false
    FlipWidthScaling=false
    OptionalParentItem=(Type=Bone,Name="None")
    RotationOffset=(X=180,Y=-40,Z=0)
    ExecutePin=@n9.ExecutePin
   )
   n12 = RigVMVariableNode(Value=@n11.OutputParam, ExecuteContext=@n10.ExecuteContext) {variable=L thighSide Alpha}
   n13 = &RigUnit_SphericalPoseReader(
    ActiveRegionScaleFactors=(PositiveWidth=3,NegativeWidth=2,PositiveHeight=2,NegativeHeight=2)
    ActiveRegionSize=0.1
    Debug=(bDrawDebug=true,bDraw2D=False,bDrawLocalAxes=False,DebugScale=25,DebugSegments=20,DebugThickness=0.25)
    DriverAxis=(X=1,Y=0,Z=0)
    DriverItem=(Type=Bone,Name="thigh_r")
    FalloffRegionScaleFactors=(PositiveWidth=2,NegativeWidth=2,PositiveHeight=2,NegativeHeight=2)
    FalloffSize=0.25
    FlipHeightScaling=False
    FlipWidthScaling=true
    OptionalParentItem=(Type=Bone,Name="None")
    RotationOffset=(X=0,Y=-40,Z=0)
    ExecutePin=@n11.ExecutePin
   )
   n14 = RigVMVariableNode(Value=@n13.OutputParam, ExecuteContext=@n12.ExecuteContext) {variable=R thighSide Alpha}
   n15 = &RigUnit_SphericalPoseReader(
    ActiveRegionScaleFactors=(PositiveWidth=2,NegativeWidth=2,PositiveHeight=2,NegativeHeight=1)
    ActiveRegionSize=0.1
    Debug=(bDrawDebug=false,bDraw2D=False,bDrawLocalAxes=False,DebugScale=25,DebugSegments=20,DebugThickness=0.25)
    DriverAxis=(X=-1,Y=0,Z=0)
    DriverItem=(Type=Bone,Name="thigh_l")
    FalloffRegionScaleFactors=(PositiveWidth=2,NegativeWidth=2,PositiveHeight=4,NegativeHeight=1)
    FalloffSize=0.3
    FlipHeightScaling=false
    FlipWidthScaling=false
    OptionalParentItem=(Type=Bone,Name="None")
    RotationOffset=(X=0,Y=90,Z=80)
    ExecutePin=@n13.ExecutePin
   )
   n16 = RigVMVariableNode(Value=@n15.OutputParam, ExecuteContext=@n14.ExecuteContext) {variable=L thighBack Alpha}
   n17 = &RigUnit_SphericalPoseReader(
    ActiveRegionScaleFactors=(PositiveWidth=2,NegativeWidth=2,PositiveHeight=2,NegativeHeight=1)
    ActiveRegionSize=0.1
    Debug=(bDrawDebug=false,bDraw2D=False,bDrawLocalAxes=False,DebugScale=25,DebugSegments=20,DebugThickness=0.25)
    DriverAxis=(X=1,Y=0,Z=0)
    DriverItem=(Type=Bone,Name="thigh_r")
    FalloffRegionScaleFactors=(PositiveWidth=2,NegativeWidth=2,PositiveHeight=3,NegativeHeight=1)
    FalloffSize=0.3
    FlipHeightScaling=false
    FlipWidthScaling=false
    OptionalParentItem=(Type=Bone,Name="None")
    RotationOffset=(X=180,Y=93.7803,Z=60.6503)
    ExecutePin=@n15.ExecutePin
   )
   n18 = RigVMVariableNode(Value=@n17.OutputParam, ExecuteContext=@n16.ExecuteContext) {variable=R thighBack Alpha}
   n19 = &RigUnit_SphericalPoseReader(
    ActiveRegionScaleFactors=(PositiveWidth=3,NegativeWidth=2,PositiveHeight=2,NegativeHeight=2)
    ActiveRegionSize=0.1
    Debug=(bDrawDebug=false,bDraw2D=False,bDrawLocalAxes=False,DebugScale=25,DebugSegments=20,DebugThickness=0.25)
    DriverAxis=(X=-1,Y=0,Z=0)
    DriverItem=(Type=Bone,Name="thigh_l")
    FalloffRegionScaleFactors=(PositiveWidth=2,NegativeWidth=2,PositiveHeight=2,NegativeHeight=2)
    FalloffSize=0.25
    FlipHeightScaling=false
    FlipWidthScaling=false
    OptionalParentItem=(Type=Bone,Name="None")
    RotationOffset=(X=2.688,Y=38.2586,Z=-2.2671)
    ExecutePin=@n17.ExecutePin
   )
   n20 = RigVMVariableNode(Value=@n19.OutputParam, ExecuteContext=@n18.ExecuteContext) {variable=L thighSideIn Alpha}
   n21 = &RigUnit_SphericalPoseReader(
    ActiveRegionScaleFactors=(PositiveWidth=3,NegativeWidth=2,PositiveHeight=2,NegativeHeight=2)
    ActiveRegionSize=0.1
    Debug=(bDrawDebug=false,bDraw2D=False,bDrawLocalAxes=False,DebugScale=25,DebugSegments=20,DebugThickness=0.25)
    DriverAxis=(X=1,Y=0,Z=0)
    DriverItem=(Type=Bone,Name="thigh_r")
    FalloffRegionScaleFactors=(PositiveWidth=2,NegativeWidth=2,PositiveHeight=2,NegativeHeight=2)
    FalloffSize=0.25
    FlipHeightScaling=False
    FlipWidthScaling=true
    OptionalParentItem=(Type=Bone,Name="None")
    RotationOffset=(X=0,Y=-40,Z=0)
    ExecutePin=@n19.ExecutePin
   )
   n22 = RigVMVariableNode(Value=@n21.OutputParam, ExecuteContext=@n20.ExecuteContext) {variable=R thighSideIn Alpha}
   n23 = &RigUnit_SphericalPoseReader(
    ActiveRegionScaleFactors=(PositiveWidth=1,NegativeWidth=1,PositiveHeight=1,NegativeHeight=1)
    ActiveRegionSize=0.2
    Debug=(bDrawDebug=false,bDraw2D=False,bDrawLocalAxes=False,DebugScale=25,DebugSegments=20,DebugThickness=0.25)
    DriverAxis=(X=1,Y=0,Z=0)
    DriverItem=(Type=Bone,Name="head")
    FalloffRegionScaleFactors=(PositiveWidth=1.5,NegativeWidth=1,PositiveHeight=1,NegativeHeight=1)
    FalloffSize=0.32
    FlipHeightScaling=False
    FlipWidthScaling=False
    OptionalParentItem=(Type=Bone,Name="None")
    RotationOffset=(X=-64.889,Y=15.1763,Z=36.1456)
    ExecutePin=@n21.ExecutePin
   )
   n24 = RigVMVariableNode(Value=@n23.OutputParam, ExecuteContext=@n22.ExecuteContext) {variable=Head Back Alpha}
   n25 = &RigUnit_SphericalPoseReader(
    ActiveRegionScaleFactors=(PositiveWidth=1,NegativeWidth=1,PositiveHeight=1,NegativeHeight=1)
    ActiveRegionSize=0.12
    Debug=(bDrawDebug=true,bDraw2D=False,bDrawLocalAxes=False,DebugScale=25,DebugSegments=20,DebugThickness=0.25)
    DriverAxis=(X=0,Y=1,Z=0)
    DriverItem=(Type=Bone,Name="thigh_l")
    FalloffRegionScaleFactors=(PositiveWidth=1,NegativeWidth=1,PositiveHeight=1,NegativeHeight=1)
    FalloffSize=0.17
    FlipHeightScaling=false
    FlipWidthScaling=false
    OptionalParentItem=(Type=Bone,Name="None")
    RotationOffset=(X=58.5993,Y=-9.6785,Z=-26.2186)
    ExecutePin=@n23.ExecutePin
   )
   n26 = RigVMVariableNode(Value=@n25.OutputParam, ExecuteContext=@n24.ExecuteContext) {variable=L thighTwistIn Alpha}
   n27 = &RigUnit_SphericalPoseReader(
    ActiveRegionScaleFactors=(PositiveWidth=1,NegativeWidth=1,PositiveHeight=1,NegativeHeight=1)
    ActiveRegionSize=0.12
    Debug=(bDrawDebug=true,bDraw2D=False,bDrawLocalAxes=False,DebugScale=25,DebugSegments=20,DebugThickness=0.25)
    DriverAxis=(X=0,Y=-1,Z=0)
    DriverItem=(Type=Bone,Name="thigh_r")
    FalloffRegionScaleFactors=(PositiveWidth=1,NegativeWidth=1,PositiveHeight=1,NegativeHeight=1)
    FalloffSize=0.17
    FlipHeightScaling=False
    FlipWidthScaling=true
    OptionalParentItem=(Type=Bone,Name="None")
    RotationOffset=(X=-127.0157,Y=4.3906,Z=-18.4663)
    ExecutePin=@n25.ExecutePin
   )
   RigVMVariableNode(Value=@n27.OutputParam, ExecuteContext=@n26.ExecuteContext) {variable=R thighTwistIn Alpha}
   n28 = &RigVMFunction_AlphaInterp(InterpSpeedDecreasing=30, InterpSpeedIncreasing=30, Scale=1, bClampResult=true, Value=RigVMVariableNode() {variable=L thighBack Alpha}.Value)
   &RigUnit_ModifyTransforms(
    ItemToModify="((Item=(Type=Bone,Name=\"dyn_skirt_bck_c_base\"),Transform=(Rotation=(X=0,Y=0,Z=0,W=-1),Translation=(X=0,Y=0,Z=1.5),Scale3D=(X=1,Y=1,Z=1))),(Item=(Type=Bone,Name=\"dyn_skirt_fr_c_base\"),Transform=(Rotation=(X=0,Y=0,Z=0,W=1),Translation=(X=0,Y=0,Z=0),Scale3D=( ...(truncated)"
    WeightMinimum=0
    Weight=@n28.Result
    ExecutePin=@RigVMFunction_Sequence.D
   )
   n29 = &RigVMFunction_AlphaInterp(InterpSpeedDecreasing=30, InterpSpeedIncreasing=30, Scale=1, bClampResult=true, Value=RigVMVariableNode() {variable=R thighBack Alpha}.Value)
   > &RigUnit_ModifyTransforms(
    ItemToModify=((Item=(Type=Bone,Name="dyn_skirt_bck_c_base"),Transform=(Rotation=(X=0,Y=0,Z=0,W=1),Translation=(X=0,Y=0,Z=1.5),Scale3D=(X=1,Y=1,Z=1))))
    WeightMinimum=0
    Weight=@n29.Result
   )
   n30 = &RigVMFunction_AlphaInterp(InterpSpeedDecreasing=20, InterpSpeedIncreasing=20, Scale=1, bClampResult=true, Value=RigVMVariableNode() {variable=Head Back Alpha}.Value)
   &RigUnit_ModifyTransforms(
    ItemToModify="((Item=(Type=Bone,Name=\"dyn_collar_side_r_base\"),Transform=(Rotation=(X=0,Y=-0.342,Z=0,W=0.9397),Translation=(X=-0.1446,Y=0,Z=0.1723),Scale3D=(X=1,Y=1,Z=1))),(Item=(Type=Bone,Name=\"dyn_collar_side_l_base\"),Transform=(Rotation=(X=0,Y=0.342,Z=0,W=-0.9397),Tr ...(truncated)"
    WeightMinimum=0
    Weight=@n30.Result
    ExecutePin=@RigVMFunction_Sequence.E
   )
   n31 = &RigVMFunction_AlphaInterp(
    InterpSpeedDecreasing=20
    InterpSpeedIncreasing=20
    Scale=1
    bClampResult=true
    Value=RigVMVariableNode() {variable=R thighTwistIn Alpha}.Value
   )
   > &RigUnit_ModifyTransforms(
    ItemToModify=((Item=(Type=Bone,Name="dyn_skirt_fr_c_base"),Transform=(Rotation=(X=0.2578,Y=0.0842,Z=-0.0226,W=-0.9623),Translation=(X=0,Y=2.1666,Z=3.7526),Scale3D=(X=1,Y=1,Z=1))))
    WeightMinimum=0
    Weight=@n31.Result
   )
   n32 = &RigVMFunction_AlphaInterp(InterpSpeedDecreasing=20, InterpSpeedIncreasing=20, Scale=1, bClampResult=true, Value=RigVMVariableNode() {variable=R thighSide Alpha}.Value)
   > &RigUnit_ModifyTransforms(
    ItemToModify=((Item=(Type=Bone,Name="dyn_skirt_fr_c_base"),Transform=(Rotation=(X=0,Y=0,Z=0,W=1),Translation=(X=0,Y=0,Z=1.6291),Scale3D=(X=1,Y=1,Z=1))))
    WeightMinimum=0
    Weight=@n32.Result
   )
   n33 = &RigVMFunction_AlphaInterp(
    InterpSpeedDecreasing=20
    InterpSpeedIncreasing=20
    Scale=1
    bClampResult=true
    Value=RigVMVariableNode() {variable=L thighTwistIn Alpha}.Value
   )
   > &RigUnit_ModifyTransforms(
    ItemToModify=((Item=(Type=Bone,Name="dyn_skirt_fr_c_base"),Transform=(Rotation=(X=0,Y=0,Z=0,W=1),Translation=(X=0,Y=0,Z=0),Scale3D=(X=1,Y=1,Z=1))))
    WeightMinimum=0
    Weight=@n33.Result
   )
   n34 = &RigVMFunction_AlphaInterp(
    InterpSpeedDecreasing=30
    InterpSpeedIncreasing=30
    Scale=1
    bClampResult=true
    Value=RigVMVariableNode() {variable=L thighForward Alpha}.Value
   )
   &RigUnit_ModifyTransforms(
    ExecutePin=()
    ItemToModify=((Item=(Type=Bone,Name="dyn_skirt_fr_c_base"),Transform=(Rotation=(X=-0.1337,Y=0.0172,Z=0.1285,W=0.9825),Translation=(X=0.3876,Y=0.9567,Z=1),Scale3D=(X=1,Y=1,Z=1))))
    WeightMinimum=0
    Weight=@n34.Result
   )
   Multiply = &RigVMFunction_MathDoubleMul(B=0.5, A=@n4.Result)
   &RigVMFunction_MathDoubleMul(B=0.5, A=@Subtract.Result)
   &RigVMFunction_MathDoubleSub(A=0, B=0)
  ]}
]
hierarchy [
 root {loc=<0 0 0>, rot=<0 0 0 1>}
 pelvis {parent=root, loc=<0 -0.5301 98.0571>, rot=<0.0127 0.707 -0.0127 -0.707>}
 spine_01 {parent=pelvis, loc=<6.1208 -0.9328 0>, rot=<0 0 0.034 -0.9994>}
 spine_02 {parent=spine_01, loc=<6.5997 -0.2095 0>, rot=<0 0 0.0238 -0.9997>}
 spine_03 {parent=spine_02, loc=<6.4914 -0.1237 0>, rot=<0 0 0.0149 -0.9999>}
 spine_04 {parent=spine_03, loc=<6.9598 0.0794 0>, rot=<0 0 -0.0132 -0.9999>}
 spine_05 {parent=spine_04, loc=<6.9987 0 0>, rot=<0 0 -0.0015 -1>}
 neck_01 {parent=spine_05, loc=<11.3138 0.1529 0>, rot=<0 0 -0.2003 -0.9797>}
 neck_02 {parent=neck_01, loc=<4.5627 0 0>, rot=<0 0 0.0504 -0.9987>}
 head {parent=neck_02, loc=<5.3914 0 0>, rot=<0 0 0.115 -0.9934>}
 clavicle_l {parent=spine_05, loc=<5.4831 -0.1053 -3.2946>, rot=<0.0465 -0.7289 -0.009 -0.6829>}
 upperarm_l {parent=clavicle_l, loc=<9.8658 -1.2136 -0.4105>, rot=<0.0164 -0.3245 -0.0269 -0.9454>}
 lowerarm_l {parent=upperarm_l, loc=<26.465 -0.8345 0.259>, rot=<0.0036 0.0161 -0.2298 -0.9731>}
 hand_l {parent=lowerarm_l, loc=<22.7355 -0.3203 -0.6105>, rot=<-0.6784 -0.0223 0.0735 -0.7307>}
 thumb_01_l {parent=hand_l, loc=<1.3118 -0.3191 -2.459>, rot=<0.6546 -0.4042 0.0664 -0.6354>}
 thumb_02_l {parent=thumb_01_l, loc=<5.9336 -0.0448 0.0401>, rot=<-0.0037 0.0213 0.298 -0.9543>}
 thumb_03_l {parent=thumb_02_l, loc=<2.2223 0.3682 -0.0907>, rot=<0.0118 -0.013 0.0066 -0.9998>}
 index_metacarpal_l {parent=hand_l, loc=<1.6535 0.0183 -2.96>, rot=<0.0158 -0.0124 -0.0173 -0.9996>}
 index_01_l {parent=index_metacarpal_l, loc=<8.3008 0.0357 -0.0524>, rot=<-0.0044 0.0485 0.0924 -0.9945>}
 index_02_l {parent=index_01_l, loc=<3.3587 -0.3004 0.0452>, rot=<-0.002 0.0027 0.3225 -0.9466>}
 index_03_l {parent=index_02_l, loc=<2.5578 -0.123 -0.0252>, rot=<-0.002 0.0067 0.2558 -0.9667>}
 middle_metacarpal_l {parent=hand_l, loc=<2.1856 -0.1725 -0.5984>, rot=<0.006 0.0168 -0.0329 -0.9993>}
 middle_01_l {parent=middle_metacarpal_l, loc=<7.3275 -0.2447 0.0129>, rot=<-0.0208 0.0042 0.0858 -0.9961>}
 middle_02_l {parent=middle_01_l, loc=<4.0123 0.1401 0.1175>, rot=<-0.0218 0.0258 0.3512 -0.9357>}
 middle_03_l {parent=middle_02_l, loc=<2.6077 0.0534 0.1373>, rot=<0.0082 0.0188 0.1708 -0.9851>}
 ring_metacarpal_l {parent=hand_l, loc=<2.2074 -0.3625 1.3638>, rot=<-0.0284 0.0722 -0.0458 -0.9959>}
 ring_01_l {parent=ring_metacarpal_l, loc=<7.0295 -0.6197 -0.1446>, rot=<-0.0489 -0.0252 0.1892 -0.9804>}
 ring_02_l {parent=ring_01_l, loc=<3.6174 0.4647 -0.1027>, rot=<0 0 0.2806 -0.9598>}
 ring_03_l {parent=ring_02_l, loc=<2.4152 0.114 0.0199>, rot=<0 0 0.1189 -0.9929>}
 pinky_metacarpal_l {parent=hand_l, loc=<1.7616 -0.9753 3.3814>, rot=<-0.0397 0.0814 -0.069 -0.9935>}
 pinky_01_l {parent=pinky_metacarpal_l, loc=<7.1709 -0.4831 -0.1094>, rot=<-0.0654 -0.0035 0.2242 -0.9723>}
 pinky_02_l {parent=pinky_01_l, loc=<2.8274 0.0893 -0.1043>, rot=<0.0037 0.0011 0.2854 -0.9584>}
 pinky_03_l {parent=pinky_02_l, loc=<2.2046 0.0307 -0.0078>, rot=<-0.0002 0.0098 0.1445 -0.9895>}
 lowerarm_twist_02_l {parent=lowerarm_l, loc=<10.6707 0 0>, rot=<0 0 0 1>}
 lowerarm_twist_01_l {parent=lowerarm_l, loc=<18.316 0 0>, rot=<0 0 0 1>}
 upperarm_twist_01_l {parent=upperarm_l, loc=<7.2135 -0.1134 0>, rot=<0 0 0 1>}
 upperarm_twist_02_l {parent=upperarm_l, loc=<19.7998 -0.1134 0>, rot=<0 0 0 1>}
 clavicle_r {parent=spine_05, loc=<5.4831 -0.1053 3.2946>, rot=<0.7289 0.0465 -0.6829 0.009>}
 upperarm_r {parent=clavicle_r, loc=<-9.8658 1.2136 0.4105>, rot=<0.0164 -0.3245 -0.0269 -0.9454>}
 lowerarm_r {parent=upperarm_r, loc=<-26.465 0.8345 -0.259>, rot=<0.0036 0.0161 -0.2298 -0.9731>}
 hand_r {parent=lowerarm_r, loc=<-22.7355 0.3203 0.6106>, rot=<-0.6784 -0.0223 0.0735 -0.7307>}
 thumb_01_r {parent=hand_r, loc=<-1.3118 0.3191 2.459>, rot=<0.6546 -0.4042 0.0664 -0.6354>}
 thumb_02_r {parent=thumb_01_r, loc=<-5.9336 0.0448 -0.0401>, rot=<-0.0037 0.0213 0.298 -0.9543>}
 thumb_03_r {parent=thumb_02_r, loc=<-2.2223 -0.3682 0.0907>, rot=<0.0118 -0.013 0.0066 -0.9998>}
 index_metacarpal_r {parent=hand_r, loc=<-1.6535 -0.0183 2.96>, rot=<0.0158 -0.0124 -0.0173 -0.9996>}
 index_01_r {parent=index_metacarpal_r, loc=<-8.3008 -0.0357 0.0524>, rot=<-0.0044 0.0485 0.0924 -0.9945>}
 index_02_r {parent=index_01_r, loc=<-3.3587 0.3004 -0.0452>, rot=<-0.002 0.0027 0.3225 -0.9466>}
 index_03_r {parent=index_02_r, loc=<-2.5578 0.123 0.0252>, rot=<-0.002 0.0067 0.2558 -0.9667>}
 middle_metacarpal_r {parent=hand_r, loc=<-2.1856 0.1725 0.5984>, rot=<0.006 0.0168 -0.0329 -0.9993>}
 middle_01_r {parent=middle_metacarpal_r, loc=<-7.3275 0.2447 -0.0129>, rot=<-0.0208 0.0042 0.0858 -0.9961>}
 middle_02_r {parent=middle_01_r, loc=<-4.0123 -0.1401 -0.1175>, rot=<-0.0218 0.0258 0.3512 -0.9357>}
 middle_03_r {parent=middle_02_r, loc=<-2.6077 -0.0534 -0.1373>, rot=<0.0082 0.0188 0.1708 -0.9851>}
 ring_metacarpal_r {parent=hand_r, loc=<-2.2074 0.3625 -1.3638>, rot=<-0.0284 0.0722 -0.0458 -0.9959>}
 ring_01_r {parent=ring_metacarpal_r, loc=<-7.0295 0.6197 0.1446>, rot=<-0.0489 -0.0252 0.1892 -0.9804>}
 ring_02_r {parent=ring_01_r, loc=<-3.6174 -0.4647 0.1027>, rot=<0 0 0.2806 -0.9598>}
 ring_03_r {parent=ring_02_r, loc=<-2.4152 -0.114 -0.0199>, rot=<0 0 0.1189 -0.9929>}
 pinky_metacarpal_r {parent=hand_r, loc=<-1.7616 0.9753 -3.3814>, rot=<-0.0397 0.0814 -0.069 -0.9935>}
 pinky_01_r {parent=pinky_metacarpal_r, loc=<-7.1709 0.4831 0.1094>, rot=<-0.0654 -0.0035 0.2242 -0.9723>}
 pinky_02_r {parent=pinky_01_r, loc=<-2.8274 -0.0893 0.1043>, rot=<0.0037 0.0011 0.2854 -0.9584>}
 pinky_03_r {parent=pinky_02_r, loc=<-2.2046 -0.0307 0.0078>, rot=<-0.0002 0.0098 0.1445 -0.9895>}
 lowerarm_twist_02_r {parent=lowerarm_r, loc=<-10.6707 0 0>, rot=<0 0 0 -1>}
 lowerarm_twist_01_r {parent=lowerarm_r, loc=<-18.316 0 0>, rot=<0 0 0 -1>}
 upperarm_twist_01_r {parent=upperarm_r, loc=<-7.2135 0.1134 0>, rot=<0 0 0 -1>}
 upperarm_twist_02_r {parent=upperarm_r, loc=<-19.7998 0.1134 0>, rot=<0 0 0 -1>}
 attach_backpack {parent=spine_05, loc=<-0.123 -13.473 0>, rot=<-0.0093 -0.707 0.0093 -0.707>}
 Backpack_BR {parent=attach_backpack, type=Null, loc=<0 0 0>, rot=<0 0 0 1>}
 attach_cape {parent=spine_05, loc=<9.017 -11.714 0>, rot=<0.0616 -0.7044 -0.0616 -0.7044>}
 cape {parent=attach_cape, type=Null, loc=<0 0 0>, rot=<0 0 0 1>}
 dyn_collar_fr_l_base {parent=spine_05, loc=<6.3806 3.3829 -4.0882>, rot=<0.057 -0.3125 -0.1981 -0.9273>}
 dyn_collar_fr_l_01 {parent=dyn_collar_fr_l_base, loc=<0 0 0>, rot=<0.0002 0.0011 -0.001 -1>}
 dyn_collar_fr_l_02 {parent=dyn_collar_fr_l_01, loc=<4.7112 0 0>, rot=<0.1293 -0.138 0.0182 -0.9818>}
 dyn_collar_side_l_base {parent=spine_05, loc=<9.0327 -0.7026 -4.5912>, rot=<0.2156 -0.1928 -0.17 -0.9421>}
 dyn_collar_side_l_01 {parent=dyn_collar_side_l_base, loc=<0 0 0>, rot=<-0.0001 0.0002 -0.0001 -1>}
 dyn_collar_side_l_02 {parent=dyn_collar_side_l_01, loc=<3.1375 -0.0657 0.3662>, rot=<0.0117 -0.1617 0.0412 -0.9859>}
 dyn_collar_bck_base {parent=spine_05, loc=<10.25 -3.5804 0>, rot=<0.711 0.0072 -0.0409 -0.7019>}
 dyn_collar_bck_01 {parent=dyn_collar_bck_base, loc=<0 0 0>, rot=<0.0004 -0.0004 0.0001 -1>}
 dyn_collar_bck_02 {parent=dyn_collar_bck_01, loc=<2.8335 -0.0993 0.2479>, rot=<0.012 -0.1372 0.0462 -0.9894>}
 dyn_collar_side_r_base {parent=spine_05, loc=<9.0327 -0.7026 4.591>, rot=<0.9158 0.1391 0.2425 -0.2882>}
 dyn_collar_side_r_01 {parent=dyn_collar_side_r_base, loc=<0 0 0>, rot=<0 0.0003 -0.0003 -1>}
 dyn_collar_side_r_02 {parent=dyn_collar_side_r_01, loc=<2.839 -0.0544 1.0143>, rot=<0.0353 -0.0425 0.0069 -0.9984>}
 dyn_collar_fr_r_base {parent=spine_05, loc=<6.3806 3.3829 4.088>, rot=<0.923 0.2056 0.3228 0.0403>}
 dyn_collar_fr_r_01 {parent=dyn_collar_fr_r_base, loc=<0 0 0>, rot=<0 0 0 1>}
 dyn_collar_fr_r_02 {parent=dyn_collar_fr_r_01, loc=<5.01 0.4632 -0.3713>, rot=<-0.1442 -0.1718 -0.0592 -0.9727>}
 thigh_l {parent=pelvis, loc=<-1.0855 -0.2109 -8.7053>, rot=<0.0009 0.0355 0.0155 -0.9993>}
 calf_l {parent=thigh_l, loc=<-42.5284 -0.0019 -0.7297>, rot=<0.0006 -0.0161 -0.0391 -0.9991>}
 foot_l {parent=calf_l, loc=<-46.7401 -2.5014 -0.592>, rot=<0.0427 -0.0238 0.0353 -0.9982>}
 ball_l {parent=foot_l, loc=<-3.5658 10.8066 -0.1544>, rot=<0.0001 -0.0001 -0.7063 -0.7079>}
 calf_twist_02_l {parent=calf_l, loc=<-19.1546 -1.7943 -0.2638>, rot=<0.0237 0.0095 -0.0035 -0.9997>}
 calf_twist_01_l {parent=calf_l, loc=<-35.8483 -1.9197 -0.5794>, rot=<0.0237 0.0095 -0.0035 -0.9997>}
 thigh_twist_01_l {parent=thigh_l, loc=<-27.952 0 0>, rot=<0.0013 0 0 -1>}
 dyn_skirt_side_l_base {parent=thigh_l, loc=<-10.3489 -2.8299 -7.4247>, rot=<-0.0061 -0.9938 0.0695 -0.0869>}
 dyn_skirt_side_l_01 {parent=dyn_skirt_side_l_base, loc=<0 0 0>, rot=<-0.0002 0.0006 -0.0025 -1>}
 dyn_skirt_side_l_02 {parent=dyn_skirt_side_l_01, loc=<11.0899 0 0>, rot=<0.0174 0.0267 -0.0033 -0.9995>}
 dyn_skirt_side_l_03 {parent=dyn_skirt_side_l_02, loc=<10.8909 0 0>, rot=<-0.099 0.0583 0.0384 -0.9926>}
 dyn_skirt_side_l_04 {parent=dyn_skirt_side_l_03, loc=<11.1861 0 0>, rot=<-0.0729 -0.0981 0.0841 -0.9889>}
 dyn_skirt_side_l_05 {parent=dyn_skirt_side_l_04, loc=<13.7575 0 0>, rot=<0.0524 -0.1421 -0.0489 -0.9873>}
 dyn_skirt_side_l_06 {parent=dyn_skirt_side_l_05, loc=<9.8291 0 0>, rot=<0.0694 -0.1043 0.0073 -0.9921>}
 thigh_r {parent=pelvis, loc=<-1.0855 -0.2109 8.7053>, rot=<0.0355 -0.0009 0.9993 0.0155>}
 calf_r {parent=thigh_r, loc=<42.5284 0.0019 0.7297>, rot=<0.0006 -0.0161 -0.0391 -0.9991>}
 foot_r {parent=calf_r, loc=<46.7401 2.5014 0.592>, rot=<0.0427 -0.0238 0.0353 -0.9982>}
 ball_r {parent=foot_r, loc=<3.5658 -10.8066 0.1544>, rot=<0.0001 -0.0001 -0.7063 -0.7079>}
 calf_twist_02_r {parent=calf_r, loc=<19.1546 1.7943 0.2638>, rot=<0.0237 0.0095 -0.0035 -0.9997>}
 calf_twist_01_r {parent=calf_r, loc=<35.8483 1.9197 0.5794>, rot=<0.0237 0.0095 -0.0035 -0.9997>}
 thigh_twist_01_r {parent=thigh_r, loc=<27.952 0 0>, rot=<0.0013 0 0 -1>}
 dyn_skirt_side_r_base {parent=thigh_r, loc=<10.3433 2.8299 7.3458>, rot=<0.9934 -0.0064 0.0913 0.0695>}
 dyn_skirt_side_r_01 {parent=dyn_skirt_side_r_base, loc=<0 0 0>, rot=<0.0003 0.0003 -0.0024 -1>}
 dyn_skirt_side_r_02 {parent=dyn_skirt_side_r_01, loc=<11.1095 0 0>, rot=<0.0174 0.0325 0.0051 -0.9993>}
 dyn_skirt_side_r_03 {parent=dyn_skirt_side_r_02, loc=<10.8721 0 0>, rot=<0.0055 -0.0405 0.018 -0.999>}
 dyn_skirt_side_r_04 {parent=dyn_skirt_side_r_03, loc=<10.8661 0 0>, rot=<0.0662 -0.1135 0.0723 -0.9887>}
 dyn_skirt_side_r_05 {parent=dyn_skirt_side_r_04, loc=<11.6944 0 0>, rot=<0.0666 0.2005 0.0369 -0.9767>}
 dyn_skirt_side_r_06 {parent=dyn_skirt_side_r_05, loc=<10.9524 0 1.1511>, rot=<0.1216 0.1724 -0.0085 -0.9775>}
 dyn_skirt_fr_c_base {parent=pelvis, loc=<-12.1701 8.15 -0.0194>, rot=<-0.0494 -0.7054 -0.7053 -0.0496>}
 dyn_skirt_fr_c_01 {parent=dyn_skirt_fr_c_base, loc=<0 0 0>, rot=<0 -0.0006 -0.0001 -1>}
 dyn_skirt_fr_c_02 {parent=dyn_skirt_fr_c_01, loc=<10.4271 0 0>, rot=<0.0392 0.0962 -0.0344 -0.994>}
 dyn_skirt_fr_c_03 {parent=dyn_skirt_fr_c_02, loc=<11.6884 0 0>, rot=<0.0159 -0.316 -0.0281 -0.9482>}
 dyn_skirt_fr_c_04 {parent=dyn_skirt_fr_c_03, loc=<8.486 0 0>, rot=<-0.0579 0.2495 0.0381 -0.9659>}
 dyn_skirt_fr_c_05 {parent=dyn_skirt_fr_c_04, loc=<9.4883 0 0>, rot=<-0.121 0.1189 0.1312 -0.9767>}
 dyn_skirt_fr_c_06 {parent=dyn_skirt_fr_c_05, loc=<9.0248 0 0>, rot=<-0.0188 -0.1853 -0.0904 -0.9783>}
 dyn_skirt_fr_c_07 {parent=dyn_skirt_fr_c_06, loc=<7.8433 0 -0.8244>, rot=<-0.0174 -0.1908 -0.0015 -0.9815>}
 dyn_skirt_bck_c_base {parent=pelvis, loc=<-5.9813 -11.8272 -0.0194>, rot=<0.0045 -0.7039 0.7092 -0.0387>}
 dyn_skirt_bck_c_01 {parent=dyn_skirt_bck_c_base, loc=<6.937 -0.3395 -0.5843>, rot=<0.0005 -0.0003 -0.0001 -1>}
 dyn_skirt_bck_c_02 {parent=dyn_skirt_bck_c_01, loc=<11.2475 0 0>, rot=<0.0347 0.0001 0.0167 -0.9993>}
 dyn_skirt_bck_c_03 {parent=dyn_skirt_bck_c_02, loc=<9.2306 0.3223 0>, rot=<-0.0113 0.2126 0.0176 -0.9769>}
 dyn_skirt_bck_c_04 {parent=dyn_skirt_bck_c_03, loc=<7.7082 0 0>, rot=<0.0259 -0.0289 -0.0751 -0.9964>}
 dyn_skirt_bck_c_05 {parent=dyn_skirt_bck_c_04, loc=<9.3588 0 -0.9837>, rot=<-0.1574 -0.2647 0.1147 -0.9445>}
 dyn_skirt_bck_c_06 {parent=dyn_skirt_bck_c_05, loc=<9.2819 -0.6554 -1.3045>, rot=<-0.0044 0.0397 -0.0552 -0.9977>}
 dyn_skirt_bck_c_07 {parent=dyn_skirt_bck_c_06, loc=<9.5328 0 0>, rot=<0.0337 0.1298 -0.0323 -0.9904>}
]
variables [
 Head Back Alpha {type=double, default=0}
 L thighBack Alpha {type=double, default=0}
 L thighForward Alpha {type=double, default=0}
 L thighSide Alpha {type=double, default=0}
 L thighSideIn Alpha {type=double, default=0}
 L thighTwistIn Alpha {type=double, default=0}
 R thighBack Alpha {type=double, default=0}
 R thighForward Alpha {type=double, default=0}
 R thighSide Alpha {type=double, default=0}
 R thighSideIn Alpha {type=double, default=0}
 R thighTwistIn Alpha {type=double, default=0}
]
