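# Physics asset for the F_MED_IvyStub female medium body cosmetic, built to simulate its dynamic skirt against the lower body.
# Capsules on spine, pelvis, thighs, calves and feet give the body shapes.
# Front and back box chains hang from the pelvis, while left and right capsule chains hang from the thighs.
# Each chain starts at a tiny, free-rotating base body, with swing and twist limits down the links, twist mostly tight outside the front.
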PhysicsAsset /BRCosmetics/Characters/Player/Female/Medium/Bodies/F_MED_IvyStub/Meshes/F_MED_IvyStub_Physics
bodies {
 calf_l [{
   shapeType=Capsule
   length=25.0642
   radius=7.2571
   shapeName=calf_l_capsule
   center=<-22.7925 -1.8153 -0.3487>
   rotation=<rot 0 90.1779 -89.1044>
  }]
 calf_r [{shapeType=Capsule, length=27.1739, radius=6.8583, shapeName=calf_r_capsule, center=<22.9672 3.0901 0.2534>, rotation=<rot 0 91.076 -89.5079>}]
 dyn_skirt_bck_c_01 [{shapeType=Box, extentX=15, extentY=24, extentZ=7, shapeName=dyn_skirt_bck_c_01_box, center=<1.6966 0 1.2511>}]
 dyn_skirt_bck_c_02 [{shapeType=Box, extentX=14, extentY=35, extentZ=9, shapeName=dyn_skirt_bck_c_02_box, center=<4.1035 0 0.2158>}]
 dyn_skirt_bck_c_03 [{shapeType=Box, extentX=10, extentY=29, extentZ=8, shapeName=dyn_skirt_bck_c_03_box, center=<1.6966 0 0.2818>}]
 dyn_skirt_bck_c_04 [{shapeType=Box, extentX=10, extentY=27, extentZ=4, shapeName=dyn_skirt_bck_c_04_box, center=<1.6966 0 0.2818>, rotation=<rot -0.0001 -10 0>}]
 dyn_skirt_bck_c_05 [{
   shapeType=Box
   extentX=10
   extentY=11
   extentZ=4
   shapeName=dyn_skirt_bck_c_05_box
   center=<1.6966 0 0.2818>
   rotation=<rot -0.0003 -0.0001 10>
  }]
 dyn_skirt_bck_c_06 [{shapeType=Box, extentX=10, extentY=11, extentZ=3, shapeName=dyn_skirt_bck_c_06_box, center=<1.6966 0 0.2818>, rotation=<rot -0.0001 0 0>}]
 dyn_skirt_bck_c_07 [{shapeType=Box, extentX=12, extentY=8, extentZ=3, shapeName=dyn_skirt_bck_c_07_box, center=<3.2399 0 0.2818>, rotation=<rot -0.0001 0 0>}]
 dyn_skirt_bck_c_base [{shapeType=Capsule, length=0.505, radius=0.505, shapeName=dyn_skirt_bck_c_base_capsule, rotation=<rot 0 -90 0>}]
 dyn_skirt_fr_c_01 [{shapeType=Box, extentX=11, extentY=33, extentZ=7, shapeName=dyn_skirt_fr_c_01_box, center=<1.8847 0 4.0536>}]
 dyn_skirt_fr_c_02 [{shapeType=Box, extentX=16, extentY=56, extentZ=13, shapeName=dyn_skirt_fr_c_02_box, center=<4.2945 0 3.0083>, rotation=<rot -19.9999 0 0>}]
 dyn_skirt_fr_c_03 [{shapeType=Box, extentX=14, extentY=26, extentZ=11, shapeName=dyn_skirt_fr_c_03_box, center=<1.7826 0 0.0238>, rotation=<rot 9.9997 0 0>}]
 dyn_skirt_fr_c_04 [{shapeType=Box, extentX=10, extentY=16, extentZ=8, shapeName=dyn_skirt_fr_c_04_box, center=<2.3186 0 0>}]
 dyn_skirt_fr_c_05 [{shapeType=Box, extentX=11, extentY=11, extentZ=5, shapeName=dyn_skirt_fr_c_05_box, center=<2.414 0 0>}]
 dyn_skirt_fr_c_06 [{shapeType=Box, extentX=10, extentY=10, extentZ=6, shapeName=dyn_skirt_fr_c_06_box, center=<1.4638 0 0>}]
 dyn_skirt_fr_c_07 [{shapeType=Box, extentX=12, extentY=9, extentZ=6, shapeName=dyn_skirt_fr_c_07_box, center=<3.3038 0 0>}]
 dyn_skirt_fr_c_base [{shapeType=Capsule, length=0.505, radius=0.505, shapeName=dyn_skirt_fr_c_base_capsule, rotation=<rot 0 -90 0>}]
 dyn_skirt_side_l_01 [{shapeType=Capsule, length=3.5, radius=8, shapeName=dyn_skirt_side_l_01_capsule, center=<5.1492 0 4.6959>, rotation=<rot -90 0.0001 0>}]
 dyn_skirt_side_l_02 [{shapeType=Capsule, length=2.5, radius=7.5, shapeName=dyn_skirt_side_l_02_capsule, center=<5.1492 0 3.8219>, rotation=<rot -90 0.0001 0>}]
 dyn_skirt_side_l_03 [{shapeType=Capsule, length=3, radius=7, shapeName=dyn_skirt_side_l_03_capsule, center=<5.1492 0 0>, rotation=<rot -90 0.0001 0>}]
 dyn_skirt_side_l_04 [{shapeType=Capsule, length=6, radius=4, shapeName=dyn_skirt_side_l_04_capsule, center=<5.1492 0 0>, rotation=<rot -90 0.0001 0>}]
 dyn_skirt_side_l_05 [{shapeType=Capsule, length=6, radius=4, shapeName=dyn_skirt_side_l_05_capsule, center=<5.1492 0 0>, rotation=<rot -90 0.0002 0>}]
 dyn_skirt_side_l_06 [{
   shapeType=Capsule
   length=6
   radius=4
   shapeName=dyn_skirt_side_l_06_capsule
   center=<7.2048 0 -1.7561>
   rotation=<rot -80 -179.9996 179.9999>
  }]
 dyn_skirt_side_l_base [{shapeType=Capsule, length=0.505, radius=0.505, shapeName=dyn_skirt_side_l_base_capsule, rotation=<rot 0 -90 0>}]
 dyn_skirt_side_r_01 [{shapeType=Capsule, length=6.5, radius=4, shapeName=dyn_skirt_side_r_01_capsule, center=<5.7417 0 1.0442>, rotation=<rot -90 -0.0002 0>}]
 dyn_skirt_side_r_02 [{shapeType=Capsule, length=5.5, radius=3.5, shapeName=dyn_skirt_side_r_02_capsule, center=<5.7417 0 0.4428>, rotation=<rot -90 -0.0002 0>}]
 dyn_skirt_side_r_03 [{shapeType=Capsule, length=5.5, radius=3.5, shapeName=dyn_skirt_side_r_03_capsule, center=<5.7417 0 0>, rotation=<rot -90 -0.0002 0>}]
 dyn_skirt_side_r_04 [{shapeType=Capsule, length=5.5, radius=3.5, shapeName=dyn_skirt_side_r_04_capsule, center=<5.7417 0 0>, rotation=<rot -90 -0.0001 0>}]
 dyn_skirt_side_r_05 [{shapeType=Capsule, length=5.5, radius=3.5, shapeName=dyn_skirt_side_r_05_capsule, center=<5.7417 0 0>, rotation=<rot -90 -0.0002 0>}]
 dyn_skirt_side_r_06 [{shapeType=Capsule, length=5.5, radius=3.5, shapeName=dyn_skirt_side_r_06_capsule, center=<5.7417 0 0>, rotation=<rot -90 -0.0002 0>}]
 dyn_skirt_side_r_base [{shapeType=Capsule, length=0.505, radius=0.505, shapeName=dyn_skirt_side_r_base_capsule, rotation=<rot 0 -90 0>}]
 foot_l [{shapeType=Capsule, length=9.0106, radius=10.0325, shapeName=foot_l_capsule, center=<-3.9219 3.9287 0.559>, rotation=<rot 0 14.9448 -82.4718>}]
 foot_r [{
   shapeType=Capsule
   length=10.6898
   radius=9.1898
   shapeName=foot_r_capsule
   center=<3.9349 -2.6216 -0.4061>
   rotation=<rot 0 8.5386 -84.2311>
  }]
 pelvis [{shapeType=Capsule, length=8.204, radius=11.0346, shapeName=pelvis_capsule, center=<-4.4759 -2.1352 0.1743>, rotation=<rot 0 175.8 -177.1898>}]
 spine_01 [{shapeType=Capsule, length=5.8449, radius=9.5068, shapeName=spine_01_capsule}]
 thigh_l [{
   shapeType=Capsule
   length=31.7359
   radius=7.7468
   shapeName=thigh_l_capsule
   center=<-24.4855 -1.5657 -0.921>
   rotation=<rot -0.4799 -85.3728 -87.2771>
  }]
 thigh_r [{
   shapeType=Capsule
   length=29.4499
   radius=7.7999
   shapeName=thigh_r_capsule
   center=<24.7888 0.8962 0.9738>
   rotation=<rot -0.4827 -83.7575 -87.2613>
  }]
}
constraints [
 {
  bone1Name=dyn_skirt_side_l_base
  bone2Name=thigh_l
  swing1LimitDegrees=45
  swing1Motion=Free
  swing2LimitDegrees=45
  swing2Motion=Free
  twistLimitDegrees=45
  twistMotion=Free
 }
 {bone1Name=dyn_skirt_side_l_01, bone2Name=dyn_skirt_side_l_base, swing1LimitDegrees=10, swing2LimitDegrees=10, twistLimitDegrees=1}
 {bone1Name=dyn_skirt_side_l_02, bone2Name=dyn_skirt_side_l_01, swing1LimitDegrees=45, swing2LimitDegrees=45, twistLimitDegrees=1}
 {bone1Name=dyn_skirt_side_l_03, bone2Name=dyn_skirt_side_l_02, swing1LimitDegrees=45, swing2LimitDegrees=45, twistLimitDegrees=1}
 {bone1Name=dyn_skirt_side_l_04, bone2Name=dyn_skirt_side_l_03, swing1LimitDegrees=45, swing2LimitDegrees=45, twistLimitDegrees=1}
 {bone1Name=dyn_skirt_side_l_05, bone2Name=dyn_skirt_side_l_04, swing1LimitDegrees=45, swing2LimitDegrees=45, twistLimitDegrees=1}
 {bone1Name=dyn_skirt_side_l_06, bone2Name=dyn_skirt_side_l_05, swing1LimitDegrees=45, swing2LimitDegrees=45, twistLimitDegrees=1}
 {
  bone1Name=dyn_skirt_side_r_base
  bone2Name=thigh_r
  swing1LimitDegrees=45
  swing1Motion=Free
  swing2LimitDegrees=45
  swing2Motion=Free
  twistLimitDegrees=45
  twistMotion=Free
 }
 {bone1Name=dyn_skirt_side_r_01, bone2Name=dyn_skirt_side_r_base, swing1LimitDegrees=10, swing2LimitDegrees=10, twistLimitDegrees=1}
 {bone1Name=dyn_skirt_side_r_02, bone2Name=dyn_skirt_side_r_01, swing1LimitDegrees=45, swing2LimitDegrees=45, twistLimitDegrees=1}
 {bone1Name=dyn_skirt_side_r_03, bone2Name=dyn_skirt_side_r_02, swing1LimitDegrees=45, swing2LimitDegrees=45, twistLimitDegrees=1}
 {bone1Name=dyn_skirt_side_r_04, bone2Name=dyn_skirt_side_r_03, swing1LimitDegrees=45, swing2LimitDegrees=45, twistLimitDegrees=1}
 {bone1Name=dyn_skirt_side_r_05, bone2Name=dyn_skirt_side_r_04, swing1LimitDegrees=45, swing2LimitDegrees=45, twistLimitDegrees=1}
 {bone1Name=dyn_skirt_side_r_06, bone2Name=dyn_skirt_side_r_05, swing1LimitDegrees=45, swing2LimitDegrees=45, twistLimitDegrees=1}
 {
  bone1Name=dyn_skirt_bck_c_base
  bone2Name=pelvis
  swing1LimitDegrees=45
  swing1Motion=Free
  swing2LimitDegrees=45
  swing2Motion=Free
  twistLimitDegrees=45
  twistMotion=Free
 }
 {
  bone1Name=dyn_skirt_fr_c_base
  bone2Name=pelvis
  swing1LimitDegrees=45
  swing1Motion=Free
  swing2LimitDegrees=45
  swing2Motion=Free
  twistLimitDegrees=45
  twistMotion=Free
 }
 {bone1Name=dyn_skirt_fr_c_01, bone2Name=dyn_skirt_fr_c_base, swing1LimitDegrees=20, swing2LimitDegrees=45, twistLimitDegrees=1}
 {bone1Name=dyn_skirt_fr_c_02, bone2Name=dyn_skirt_fr_c_01, swing1LimitDegrees=20, swing2LimitDegrees=45, twistLimitDegrees=30}
 {bone1Name=dyn_skirt_fr_c_03, bone2Name=dyn_skirt_fr_c_02, swing1LimitDegrees=45, swing2LimitDegrees=60, twistLimitDegrees=15}
 {bone1Name=dyn_skirt_fr_c_04, bone2Name=dyn_skirt_fr_c_03, swing1LimitDegrees=45, swing2LimitDegrees=60, twistLimitDegrees=15}
 {bone1Name=dyn_skirt_fr_c_05, bone2Name=dyn_skirt_fr_c_04, swing1LimitDegrees=45, swing2LimitDegrees=60, twistLimitDegrees=15}
 {bone1Name=dyn_skirt_fr_c_06, bone2Name=dyn_skirt_fr_c_05, swing1LimitDegrees=45, swing2LimitDegrees=60, twistLimitDegrees=15}
 {bone1Name=dyn_skirt_fr_c_07, bone2Name=dyn_skirt_fr_c_06, swing1LimitDegrees=45, swing2LimitDegrees=60, twistLimitDegrees=15}
 {bone1Name=dyn_skirt_bck_c_01, bone2Name=dyn_skirt_bck_c_base, swing1LimitDegrees=45, swing2LimitDegrees=45, twistLimitDegrees=1}
 {bone1Name=dyn_skirt_bck_c_02, bone2Name=dyn_skirt_bck_c_01, swing1LimitDegrees=45, swing2LimitDegrees=45, twistLimitDegrees=1}
 {bone1Name=dyn_skirt_bck_c_03, bone2Name=dyn_skirt_bck_c_02, swing1LimitDegrees=45, swing2LimitDegrees=45, twistLimitDegrees=1}
 {bone1Name=dyn_skirt_bck_c_04, bone2Name=dyn_skirt_bck_c_03, swing1LimitDegrees=45, swing2LimitDegrees=45, twistLimitDegrees=1}
 {bone1Name=dyn_skirt_bck_c_05, bone2Name=dyn_skirt_bck_c_04, swing1LimitDegrees=45, swing2LimitDegrees=45, twistLimitDegrees=1}
 {bone1Name=dyn_skirt_bck_c_06, bone2Name=dyn_skirt_bck_c_05, swing1LimitDegrees=45, swing2LimitDegrees=45, twistLimitDegrees=1}
 {bone1Name=dyn_skirt_bck_c_07, bone2Name=dyn_skirt_bck_c_06, swing1LimitDegrees=45, swing2LimitDegrees=45, twistLimitDegrees=1}
]
