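# Animation Blueprint for the F_MED_IvyStub female body cosmetic, adding simulated secondary motion to the player's copied pose.
# The parent mesh pose feeds a Control Rig, a pelvis-space RigidBody simulation, then a post-sim Control Rig.
# The post-sim rig receives skydive and slide states, the IvyStub emote curve, and an alt-skin flag inverted from bRigidBodyEnabled.
# Skydive and parachute events swap the simulation's gravity vector; the gravity parameter set goes unread in these graphs.
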AnimBlueprint /BRCosmetics/Characters/Player/Female/Medium/Bodies/F_MED_IvyStub/Meshes/F_MED_IvyStub_AnimBP
target_skeleton=&/F_MED_IvyStub_Skeleton
graphs [
 AnimGraph {nodes [
   n1 = Control Rig(Source=Copy Pose From Mesh() {bCopyCurves=true, bUseAttachedParent=true}) {
    alphaBoolBlend {blendOption=Linear}
    controlRigAssetReference {blueprintRigClass=&/F_MED_IvyStub_CtrlRig.F_MED_IvyStub_CtrlRig_C}
   }
   n2 = RigidBody!(
    ComponentPose=Local To Component(LocalPose=@n1)
    Alpha
    OverrideWorldGravity:Vector=Get Override World Gravity() -> (Override World Gravity:Vector)
    ExternalForce
   ) -> (Pose) {
    bClampLinearTranslationLimitToRefPose=true
    bOverrideWorldGravity=true
    overridePhysicsAsset=&/F_MED_IvyStub_Physics
    simulationSpace=BaseBoneSpace
    alphaBoolBlend {blendOption=Linear}
    baseBoneRef {boneName=pelvis}
    componentAppliedLinearAccClamp=<3000 3000 3000>
    overrideWorldGravity {z=-980}
    simSpaceSettings {maxAngularAcceleration=800, maxAngularVelocity=8, maxLinearAcceleration=800, maxLinearVelocity=800, worldAlpha=0.17}
   }
   n3 = Control Rig!(
    Source=Component To Local(ComponentPose=@n2)
    Alpha
    Is Skydiving Dive Mode:bool=Get bIsSkydiveDiveMode()
    Is Skydiving:bool=Get bIsSkyDiving()
    IsBaseballSliding:bool=Get bIsBaseballSliding()
    IsKneeSliding:bool=Get bIsKneeSliding()
    Custom Emote Alpha:double=Get CustomEmote() -> (CustomEmote:double)
    isALT skin:bool=Get isALT skin() -> (isALT skin:bool)
   ) -> (Pose) {
    alphaBoolBlend {blendOption=Linear}
    controlRigAssetReference {blueprintRigClass=&/F_MED_IvyStub_PostSim_CtrlRig.F_MED_IvyStub_PostSim_CtrlRig_C}
   }
   Output Pose(Result=@n3) {layerGroup=None}
  ]}
 EventGraph {nodes [
   Event BlueprintUpdateAnimation()
   > Set CustomEmote(CustomEmote:double=GetCurveValue(CurveName=IvyStub)) -> (Output_Get:double)
   > Set isALT skin(isALT skin:bool=NOT Boolean(A=Get bRigidBodyEnabled())) -> (Output_Get:bool)
   n1 = Event OnBeginSkydiving()
   Set Override World Gravity(execute=[@n1.then, @n2.then], Override World Gravity:Vector="0,0,500") -> (Output_Get:Vector)
   Event OnEndSkydiving()
   > Set Override World Gravity(Override World Gravity="0,0,-980")
   Event OnParachuteOpened()
   > Set Override World Gravity(Override World Gravity="200,0,100")
   Event OnParachuteClosed()
   > n2 = Branch(Condition=Get bIsSkyDiving())
   Set Override World Gravity(execute=@n2.else, Override World Gravity="0,0,-980")
  ]}
]
variables [
 CustomEmote {type=number}
 Gravity Override Params {type=GravityOverrideParamsStruct}
 Gravity Override Params_0 {
  type=FortGravityOverrideParameterSet
  default=(Skydive_GoingUp=(GravityOverride=(WindFrequency=(X=1.500000,Y=7.000000,Z=8.000000),WindAmplitude=(X=200.000000,Y=-500.000000,Z=400.000000),GravityOverride=(X=0.000000,Y=-300.000000,Z=1500.000000),JointName="root")),Skydive_FloatingDown=(GravityOverride=(WindFrequency=(X=1.500000,Y=7.000000,Z=8.000000),WindAmplitude=(X=200.000000,Y=-500.000000,Z=400.000000),WindOffset=(X=0.000000,Y=-600.000000,Z=400.000000),GravityOverride=(X=0.000000,Y=-700.000000,Z=600.000000),JointName="root")),Skydive_DivingDown=(GravityOverride=(WindFrequency=(X=1.500000,Y=12.000000,Z=10.000000),WindAmplitude=(X=50.000000,Y=300.000000,Z=100.000000),GravityOverride=(X=-1500.000000,Y=0.000000,Z=0.000000),JointName="pelvis")),Skydive_Parachuting=(GravityOverride=(WindFrequency=(X=1.500000,Y=7.000000,Z=8.000000),WindAmplitude=(X=200.000000,Y=-500.000000,Z=400.000000),GravityOverride=(X=0.000000,Y=-800.000000,Z=-980.000000),JointName="root")),OnGround_Standing=(GravityOverride=(GravityOverride=(X=0.000000,Y=0.000000,Z=-980.000000),JointName="root")),OnGround_Moving=(GravityOverride=(GravityOverride=(X=0.000000,Y=0.000000,Z=-980.000000),JointName="root")),OnGround_Moving_HighRotationRate=(GravityOverride=(GravityOverride=(X=0.000000,Y=0.000000,Z=-980.000000),JointName="root")),OnGround_Crouch_Moving=(GravityOverride=()),Emote_or_Melee=(GravityOverride=()),Swimming=(GravityOverride=(GravityOverride=(X=-800.000000,Y=0.000000,Z=0.000000),JointName="pelvis")),Front_End=(GravityOverride=(GravityOverride=(X=0.000000,Y=0.000000,Z=-980.000000),JointName="root")),DBNO=(GravityOverride=(GravityOverride=(X=-980.000000,Y=400.000000,Z=0.000000),JointName="pelvis")),DBNO_Carried=(GravityOverride=(GravityOverride=(X=-980.000000,Y=500.000000,Z=-500.000000),JointName="pelvis")),OnGround_Falling=(GravityOverride=(GravityOverride=(X=0.000000,Y=0.000000,Z=-980.000000),JointName="root")),Ziplining=(GravityOverride=(GravityOverride=(X=0.000000,Y=0.000000,Z=-980.000000),JointName="root")),Skydive_Surfing=(GravityOverride=(WindFrequency=(X=1.500000,Y=7.000000,Z=8.000000),WindAmplitude=(X=200.000000,Y=-500.000000,Z=400.000000),GravityOverride=(X=500.000000,Y=-300.000000,Z=-980.000000),JointName="root")),Riding_Standing=(GravityOverride=(GravityOverride=(X=0.000000,Y=-350.000000,Z=-980.000000),JointName="root")),Riding_Moving=(GravityOverride=(GravityOverride=(X=0.000000,Y=-450.000000,Z=-450.000000),JointName="root")),Riding_Moving_Targeting=(GravityOverride=(GravityOverride=(X=0.000000,Y=-200.000000,Z=-980.000000),JointName="root")),Tactical_Sprint=(GravityOverride=(GravityOverride=(X=0.000000,Y=-450.000000,Z=-500.000000),JointName="root")),Rail_Grinding=(GravityOverride=(GravityOverride=(X=0.000000,Y=0.000000,Z=-980.000000),JointName="root")),)
 }
 isALT skin {type=boolean}
 Override World Gravity {type=Vector, default=<0 0 -980>}
 sineRotation high1 {type=Rotator}
 sineRotation high2 {type=Rotator}
 sineRotation high3 {type=Rotator}
 sineRotation1 {type=Rotator}
 sineRotation2 {type=Rotator}
 sineRotation3 {type=Rotator}
 sineRotation4 {type=Rotator}
 sineRotation5 {type=Rotator}
]
