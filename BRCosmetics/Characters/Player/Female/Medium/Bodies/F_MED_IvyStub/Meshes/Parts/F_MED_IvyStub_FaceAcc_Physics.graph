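# Physics asset for the F_MED_IvyStub female character's FaceAcc mesh, providing capsule bodies and joints for its dynamic hair strands.
# Lettered upper and lower strand chains hang from the head or from left, middle and right master hair bodies, with swing limits and very tight twist.
# Larger head, neck, upper spine, clavicle and upper-arm capsules cover the body around the hair, linked by free angular joints.
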PhysicsAsset /BRCosmetics/Characters/Player/Female/Medium/Bodies/F_MED_IvyStub/Meshes/Parts/F_MED_IvyStub_FaceAcc_Physics
bodies {
 clavicle_l [{shapeType=Capsule, length=4, radius=6, shapeName=clavicle_l_capsule, center=<3.2872 0 -1.6368>, rotation=<rot -90 -0.0001 0>}]
 clavicle_r [{shapeType=Capsule, length=4, radius=6, shapeName=clavicle_r_capsule, center=<-3.2872 0 1.6368>, rotation=<rot 90 0.0001 0>}]
 dyn_hairStrand_lower_A_01 [{shapeType=Capsule, length=4, radius=3, shapeName=dyn_hairStrand_lower_A_01_capsule, center=<1.5271 0 0>, rotation=<rot 90 -0.0003 0>}]
 dyn_hairStrand_lower_A_02 [{shapeType=Capsule, length=4, radius=3, shapeName=dyn_hairStrand_lower_A_02_capsule, center=<1.5271 0 0>, rotation=<rot 90 -0.0002 0>}]
 dyn_hairStrand_lower_A_03 [{shapeType=Capsule, length=4, radius=3, shapeName=dyn_hairStrand_lower_A_03_capsule, center=<1.5271 0 0>, rotation=<rot 90 0.0001 0>}]
 dyn_hairStrand_lower_B_01 [{shapeType=Capsule, length=4, radius=3, shapeName=dyn_hairStrand_lower_B_01_capsule, center=<1.5271 0 0>, rotation=<rot 90 -0.0003 0>}]
 dyn_hairStrand_lower_B_02 [{shapeType=Capsule, length=4, radius=3, shapeName=dyn_hairStrand_lower_B_02_capsule, center=<1.5271 0 0>, rotation=<rot 90 -0.0002 0>}]
 dyn_hairStrand_lower_B_03 [{shapeType=Capsule, length=4, radius=3, shapeName=dyn_hairStrand_lower_B_03_capsule, center=<1.5271 0 0>, rotation=<rot 90 0.0001 0>}]
 dyn_hairStrand_lower_C_01 [{shapeType=Capsule, length=4, radius=3, shapeName=dyn_hairStrand_lower_C_01_capsule, center=<1.5271 0 0>, rotation=<rot 70 179.9996 179.9999>}]
 dyn_hairStrand_lower_C_02 [{shapeType=Capsule, length=4, radius=3, shapeName=dyn_hairStrand_lower_C_02_capsule, center=<1.5271 0 0>, rotation=<rot 90 -0.0002 0>}]
 dyn_hairStrand_lower_D_01 [{shapeType=Capsule, length=4, radius=3, shapeName=dyn_hairStrand_lower_D_01_capsule, center=<1.5271 0 0>, rotation=<rot 90 -0.0003 0>}]
 dyn_hairStrand_lower_D_02 [{shapeType=Capsule, length=4, radius=3, shapeName=dyn_hairStrand_lower_D_02_capsule, center=<1.5271 0 0>, rotation=<rot 90 -0.0002 0>}]
 dyn_hairStrand_lower_D_03 [{shapeType=Capsule, length=4, radius=3, shapeName=dyn_hairStrand_lower_D_03_capsule, center=<1.5271 0 0>, rotation=<rot 90 0.0001 0>}]
 dyn_hairStrand_lower_E_01 [{shapeType=Capsule, length=4, radius=3, shapeName=dyn_hairStrand_lower_E_01_capsule, center=<1.5271 0 0>, rotation=<rot 90 -0.0002 0>}]
 dyn_hairStrand_lower_E_02 [{shapeType=Capsule, length=4, radius=3, shapeName=dyn_hairStrand_lower_E_02_capsule, center=<1.5271 0 0>, rotation=<rot 90 -0.0001 0>}]
 dyn_hairStrand_lower_F_01 [{shapeType=Capsule, length=4, radius=3, shapeName=dyn_hairStrand_lower_F_01_capsule, center=<1.5271 0 0>, rotation=<rot 90 -0.0002 0>}]
 dyn_hairStrand_lower_F_02 [{shapeType=Capsule, length=4, radius=3, shapeName=dyn_hairStrand_lower_F_02_capsule, center=<1.7211 0 0>, rotation=<rot 90 0 0>}]
 dyn_hairStrand_lower_G_01 [{shapeType=Capsule, length=4, radius=3, shapeName=dyn_hairStrand_lower_G_01_capsule, center=<1.5271 0 0>, rotation=<rot 90 -0.0002 0>}]
 dyn_hairStrand_lower_G_02 [{shapeType=Capsule, length=4, radius=3, shapeName=dyn_hairStrand_lower_G_02_capsule, center=<1.5271 0 0>, rotation=<rot 90 -0.0001 0>}]
 dyn_hairStrand_lower_G_03 [{shapeType=Capsule, length=4, radius=3, shapeName=dyn_hairStrand_lower_G_03_capsule, center=<1.7211 0 0>, rotation=<rot 90 0.0001 0>}]
 dyn_hairStrand_lower_G_04 [{shapeType=Capsule, length=4, radius=3, shapeName=dyn_hairStrand_lower_G_04_capsule, center=<1.5271 0 0>, rotation=<rot 90 0 0>}]
 dyn_hairStrand_lower_G_05 [{shapeType=Capsule, length=4, radius=3, shapeName=dyn_hairStrand_lower_G_05_capsule, center=<1.5271 0 0>, rotation=<rot 90 0 0>}]
 dyn_hairStrand_lower_H [{shapeType=Capsule, length=4, radius=3, shapeName=dyn_hairStrand_lower_H_capsule, center=<1.5271 0 0>, rotation=<rot 90 0.0001 0>}]
 dyn_hairStrand_lower_I_01 [{shapeType=Capsule, length=4, radius=3, shapeName=dyn_hairStrand_lower_I_01_capsule, center=<1.5271 0 0>, rotation=<rot 90 0.0001 0>}]
 dyn_hairStrand_lower_I_02 [{shapeType=Capsule, length=4, radius=3, shapeName=dyn_hairStrand_lower_I_02_capsule, center=<1.5271 0 0>, rotation=<rot 90 -0.0001 0>}]
 dyn_hairStrand_lower_I_03 [{shapeType=Capsule, length=4, radius=3, shapeName=dyn_hairStrand_lower_I_03_capsule, center=<1.5271 0 0>, rotation=<rot 90 0 0>}]
 dyn_hairStrand_lower_J_01 [{shapeType=Capsule, length=4, radius=3, shapeName=dyn_hairStrand_lower_J_01_capsule, center=<1.5271 0 0>, rotation=<rot 90 0.0002 0>}]
 dyn_hairStrand_lower_J_02 [{shapeType=Capsule, length=4, radius=3, shapeName=dyn_hairStrand_lower_J_02_capsule, center=<1.5271 0 0>, rotation=<rot 90 0 0>}]
 dyn_hairStrand_lower_K_01 [{shapeType=Capsule, length=4, radius=3, shapeName=dyn_hairStrand_lower_K_01_capsule, center=<1.5271 0 0>, rotation=<rot 90 0.0002 0>}]
 dyn_hairStrand_lower_K_02 [{
   shapeType=Capsule
   length=4
   radius=3
   shapeName=dyn_hairStrand_lower_K_02_capsule
   center=<1.7031 0 0.8647>
   rotation=<rot 80 -0.0001 -0.0002>
  }]
 dyn_hairStrand_lower_L_01 [{shapeType=Capsule, length=4, radius=3, shapeName=dyn_hairStrand_lower_L_01_capsule, center=<1.5271 0 0>, rotation=<rot 90 0.0002 0>}]
 dyn_hairStrand_lower_L_02 [{shapeType=Capsule, length=4, radius=3, shapeName=dyn_hairStrand_lower_L_02_capsule, center=<1.5271 0 0>, rotation=<rot 90 0.0001 0>}]
 dyn_hairStrand_lower_M_01 [{shapeType=Capsule, length=4, radius=3, shapeName=dyn_hairStrand_lower_M_01_capsule, center=<1.5271 0 0>, rotation=<rot 90 0.0003 0>}]
 dyn_hairStrand_lower_M_02 [{shapeType=Capsule, length=4, radius=3, shapeName=dyn_hairStrand_lower_M_02_capsule, center=<1.5271 0 0>, rotation=<rot 90 0.0003 0>}]
 dyn_hairStrand_lower_M_03 [{shapeType=Capsule, length=4, radius=3, shapeName=dyn_hairStrand_lower_M_03_capsule, center=<1.5271 0 0>, rotation=<rot 90 0.0002 0>}]
 dyn_hairStrand_lower_M_04 [{shapeType=Capsule, length=4, radius=3, shapeName=dyn_hairStrand_lower_M_04_capsule, center=<1.5271 0 0>, rotation=<rot 90 0 0>}]
 dyn_hairStrand_lower_N_01 [{shapeType=Capsule, length=4, radius=3, shapeName=dyn_hairStrand_lower_N_01_capsule, center=<1.5271 0 0>, rotation=<rot 90 0.0002 0>}]
 dyn_hairStrand_lower_N_02 [{shapeType=Capsule, length=4, radius=3, shapeName=dyn_hairStrand_lower_N_02_capsule, center=<1.5271 0 0>, rotation=<rot 90 0.0003 0>}]
 dyn_hairStrand_lower_N_03 [{shapeType=Capsule, length=4, radius=3, shapeName=dyn_hairStrand_lower_N_03_capsule, center=<1.5271 0 0>, rotation=<rot 90 0.0001 0>}]
 dyn_hairStrand_upper_A_01 [{shapeType=Capsule, length=3, radius=2, shapeName=dyn_hairStrand_upper_A_01_capsule, center=<1.2288 0 0>, rotation=<rot 90 -0.0004 0>}]
 dyn_hairStrand_upper_A_02 [{shapeType=Capsule, length=3, radius=2, shapeName=dyn_hairStrand_upper_A_02_capsule, center=<1.2288 0 0>, rotation=<rot 90 -0.0002 0>}]
 dyn_hairStrand_upper_B_01 [{shapeType=Capsule, length=3, radius=2, shapeName=dyn_hairStrand_upper_B_01_capsule, center=<1.2288 0 0>, rotation=<rot 90 -0.0004 0>}]
 dyn_hairStrand_upper_B_02 [{shapeType=Capsule, length=3, radius=2, shapeName=dyn_hairStrand_upper_B_02_capsule, center=<1.2288 0 0>, rotation=<rot 90 0.0005 0>}]
 dyn_hairStrand_upper_D_01 [{shapeType=Capsule, length=3, radius=2, shapeName=dyn_hairStrand_upper_D_01_capsule, center=<1.6033 0 0>, rotation=<rot 90 -0.0002 0>}]
 dyn_hairStrand_upper_D_02 [{shapeType=Capsule, length=3, radius=2, shapeName=dyn_hairStrand_upper_D_02_capsule, center=<1.6033 0 0>, rotation=<rot 90 0 0>}]
 dyn_hairStrand_upper_E_01 [{
   shapeType=Capsule
   length=3.1983
   radius=1.2366
   shapeName=dyn_hairStrand_upper_E_01_capsule
   center=<2.832 -0.0391 0.043>
   rotation=<rot 0 91.3715 -77.3261>
  }]
 dyn_hairStrand_upper_E_02 [{
   shapeType=Capsule
   length=1.9939
   radius=1.3004
   shapeName=dyn_hairStrand_upper_E_02_capsule
   center=<2.2558 -0.0934 0.3434>
   rotation=<rot 0 129.6157 -79.7399>
  }]
 dyn_hairStrand_upper_E_03 [{
   shapeType=Capsule
   length=2.9854
   radius=1.0999
   shapeName=dyn_hairStrand_upper_E_03_capsule
   center=<2.5936 -0.2839 1.4671>
   rotation=<rot -12.434 -90 34.6968>
  }]
 dyn_hairStrand_upper_F_01 [{
   shapeType=Capsule
   length=2.5375
   radius=1.2647
   shapeName=dyn_hairStrand_upper_F_01_capsule
   center=<2.5113 -0.2473 -0.0489>
   rotation=<rot 0 104.4737 -88.9686>
  }]
 dyn_hairStrand_upper_F_02 [{
   shapeType=Capsule
   length=2.6526
   radius=1.4052
   shapeName=dyn_hairStrand_upper_F_02_capsule
   center=<1.9415 0.1354 1.4334>
   rotation=<rot 0.4526 -90 25.4001>
  }]
 dyn_hairStrand_upper_H_01 [{shapeType=Capsule, length=3, radius=2, shapeName=dyn_hairStrand_upper_H_01_capsule, center=<1.6033 0 0>, rotation=<rot 90 0.0002 0>}]
 dyn_hairStrand_upper_I_01 [{shapeType=Capsule, length=3, radius=2, shapeName=dyn_hairStrand_upper_I_01_capsule, center=<1.6033 0 0>, rotation=<rot 90 0.0003 0>}]
 dyn_hairStrand_upper_J_01 [{shapeType=Capsule, length=3, radius=2, shapeName=dyn_hairStrand_upper_J_01_capsule, center=<1.2288 0 0>, rotation=<rot 90 0.0001 0>}]
 dyn_hairStrand_upper_K_01 [{shapeType=Capsule, length=3, radius=2, shapeName=dyn_hairStrand_upper_K_01_capsule, center=<1.2288 0 0>, rotation=<rot 90 -0.0003 0>}]
 dyn_hair_Mstr_l [{shapeType=Capsule, length=2, radius=2, shapeName=dyn_hair_Mstr_l_capsule, center=<0 0 -1.5271>}]
 dyn_hair_Mstr_mid [{shapeType=Capsule, length=2, radius=2, shapeName=dyn_hair_Mstr_mid_capsule, center=<0 0 -1.5271>}]
 dyn_hair_Mstr_r [{shapeType=Capsule, length=2, radius=2, shapeName=dyn_hair_Mstr_r_capsule, center=<0 0 -1.5271>}]
 head [{shapeType=Capsule, length=2, radius=10.5, shapeName=head_capsule, center=<2.6993 -0.0997 0>, rotation=<rot 90 -0.0002 0>}]
 neck_01 [{shapeType=Capsule, length=0, radius=5, shapeName=neck_01_capsule}]
 neck_02 [{shapeType=Capsule, length=0, radius=4.5, shapeName=neck_02_capsule}]
 spine_05 [{shapeType=Capsule, length=11, radius=8.5, shapeName=spine_05_capsule, center=<3.8597 0 0>}]
 upperarm_l [{shapeType=Capsule, length=13, radius=5, shapeName=upperarm_l_capsule, center=<9.6905 0 -1.4595>, rotation=<rot -90 0 0>}]
 upperarm_r [{shapeType=Capsule, length=13, radius=5, shapeName=upperarm_r_capsule, center=<-9.6905 0 1.4595>, rotation=<rot 90 0.0002 0>}]
}
constraints [
 {
  bone1Name=head
  bone2Name=neck_02
  swing1LimitDegrees=45
  swing1Motion=Free
  swing2LimitDegrees=45
  swing2Motion=Free
  twistLimitDegrees=45
  twistMotion=Free
 }
 {bone1Name=dyn_hair_Mstr_l, bone2Name=head, swing1LimitDegrees=45, swing2LimitDegrees=45, twistLimitDegrees=1}
 {bone1Name=dyn_hair_Mstr_mid, bone2Name=head, swing1LimitDegrees=45, swing2LimitDegrees=45, twistLimitDegrees=1}
 {bone1Name=dyn_hair_Mstr_r, bone2Name=head, swing1LimitDegrees=45, swing2LimitDegrees=45, twistLimitDegrees=1}
 {
  bone1Name=neck_01
  bone2Name=spine_05
  swing1LimitDegrees=45
  swing1Motion=Free
  swing2LimitDegrees=45
  swing2Motion=Free
  twistLimitDegrees=45
  twistMotion=Free
 }
 {
  bone1Name=neck_02
  bone2Name=spine_05
  swing1LimitDegrees=45
  swing1Motion=Free
  swing2LimitDegrees=45
  swing2Motion=Free
  twistLimitDegrees=45
  twistMotion=Free
 }
 {
  bone1Name=clavicle_l
  bone2Name=spine_05
  swing1LimitDegrees=45
  swing1Motion=Free
  swing2LimitDegrees=45
  swing2Motion=Free
  twistLimitDegrees=45
  twistMotion=Free
 }
 {
  bone1Name=upperarm_l
  bone2Name=spine_05
  swing1LimitDegrees=45
  swing1Motion=Free
  swing2LimitDegrees=45
  swing2Motion=Free
  twistLimitDegrees=45
  twistMotion=Free
 }
 {
  bone1Name=clavicle_r
  bone2Name=spine_05
  swing1LimitDegrees=45
  swing1Motion=Free
  swing2LimitDegrees=45
  swing2Motion=Free
  twistLimitDegrees=45
  twistMotion=Free
 }
 {
  bone1Name=upperarm_r
  bone2Name=spine_05
  swing1LimitDegrees=45
  swing1Motion=Free
  swing2LimitDegrees=45
  swing2Motion=Free
  twistLimitDegrees=45
  twistMotion=Free
 }
 {
  bone1Name=dyn_hair_Mstr_l
  bone2Name=dyn_hair_Mstr_mid
  swing1LimitDegrees=45
  swing1Motion=Free
  swing2LimitDegrees=45
  swing2Motion=Free
  twistLimitDegrees=45
  twistMotion=Free
 }
 {
  bone1Name=dyn_hair_Mstr_r
  bone2Name=dyn_hair_Mstr_mid
  swing1LimitDegrees=45
  swing1Motion=Free
  swing2LimitDegrees=45
  swing2Motion=Free
  twistLimitDegrees=45
  twistMotion=Free
 }
 {bone1Name=dyn_hairStrand_lower_A_01, bone2Name=head, swing1LimitDegrees=45, swing2LimitDegrees=45, twistLimitDegrees=1}
 {bone1Name=dyn_hairStrand_lower_A_02, bone2Name=dyn_hairStrand_lower_A_01, swing1LimitDegrees=45, swing2LimitDegrees=45, twistLimitDegrees=1}
 {bone1Name=dyn_hairStrand_lower_A_03, bone2Name=dyn_hairStrand_lower_A_02, swing1LimitDegrees=45, swing2LimitDegrees=45, twistLimitDegrees=1}
 {bone1Name=dyn_hairStrand_lower_M_01, bone2Name=head, swing1LimitDegrees=45, swing2LimitDegrees=45, twistLimitDegrees=1}
 {bone1Name=dyn_hairStrand_lower_M_02, bone2Name=dyn_hairStrand_lower_M_01, swing1LimitDegrees=45, swing2LimitDegrees=45, twistLimitDegrees=1}
 {bone1Name=dyn_hairStrand_lower_M_03, bone2Name=dyn_hairStrand_lower_M_02, swing1LimitDegrees=45, swing2LimitDegrees=45, twistLimitDegrees=1}
 {bone1Name=dyn_hairStrand_lower_M_04, bone2Name=dyn_hairStrand_lower_M_03, swing1LimitDegrees=45, swing2LimitDegrees=45, twistLimitDegrees=1}
 {bone1Name=dyn_hairStrand_lower_N_01, bone2Name=head, swing1LimitDegrees=45, swing2LimitDegrees=45, twistLimitDegrees=1}
 {bone1Name=dyn_hairStrand_lower_N_02, bone2Name=dyn_hairStrand_lower_N_01, swing1LimitDegrees=45, swing2LimitDegrees=45, twistLimitDegrees=1}
 {bone1Name=dyn_hairStrand_lower_N_03, bone2Name=dyn_hairStrand_lower_N_02, swing1LimitDegrees=45, swing2LimitDegrees=45, twistLimitDegrees=1}
 {bone1Name=dyn_hairStrand_lower_J_01, bone2Name=dyn_hair_Mstr_l, swing1LimitDegrees=45, swing2LimitDegrees=45, twistLimitDegrees=1}
 {bone1Name=dyn_hairStrand_lower_J_02, bone2Name=dyn_hairStrand_lower_J_01, swing1LimitDegrees=45, swing2LimitDegrees=45, twistLimitDegrees=1}
 {bone1Name=dyn_hairStrand_lower_K_01, bone2Name=dyn_hair_Mstr_l, swing1LimitDegrees=45, swing2LimitDegrees=45, twistLimitDegrees=1}
 {bone1Name=dyn_hairStrand_lower_K_02, bone2Name=dyn_hairStrand_lower_K_01, swing1LimitDegrees=45, swing2LimitDegrees=45, twistLimitDegrees=1}
 {bone1Name=dyn_hairStrand_lower_L_01, bone2Name=dyn_hair_Mstr_l, swing1LimitDegrees=45, swing2LimitDegrees=45, twistLimitDegrees=1}
 {bone1Name=dyn_hairStrand_lower_L_02, bone2Name=dyn_hairStrand_lower_L_01, swing1LimitDegrees=45, swing2LimitDegrees=45, twistLimitDegrees=1}
 {bone1Name=dyn_hairStrand_lower_G_01, bone2Name=dyn_hair_Mstr_mid, swing1LimitDegrees=45, swing2LimitDegrees=45, twistLimitDegrees=1}
 {bone1Name=dyn_hairStrand_lower_G_02, bone2Name=dyn_hairStrand_lower_G_01, swing1LimitDegrees=45, swing2LimitDegrees=45, twistLimitDegrees=1}
 {bone1Name=dyn_hairStrand_lower_G_03, bone2Name=dyn_hairStrand_lower_G_02, swing1LimitDegrees=45, swing2LimitDegrees=45, twistLimitDegrees=1}
 {bone1Name=dyn_hairStrand_lower_G_04, bone2Name=dyn_hairStrand_lower_G_02, swing1LimitDegrees=45, swing2LimitDegrees=45, twistLimitDegrees=1}
 {bone1Name=dyn_hairStrand_lower_G_05, bone2Name=dyn_hairStrand_lower_G_02, swing1LimitDegrees=45, swing2LimitDegrees=45, twistLimitDegrees=1}
 {bone1Name=dyn_hairStrand_lower_B_01, bone2Name=dyn_hair_Mstr_r, swing1LimitDegrees=45, swing2LimitDegrees=45, twistLimitDegrees=1}
 {bone1Name=dyn_hairStrand_lower_B_02, bone2Name=dyn_hairStrand_lower_B_01, swing1LimitDegrees=45, swing2LimitDegrees=45, twistLimitDegrees=1}
 {bone1Name=dyn_hairStrand_lower_B_03, bone2Name=dyn_hairStrand_lower_B_02, swing1LimitDegrees=45, swing2LimitDegrees=45, twistLimitDegrees=1}
 {bone1Name=dyn_hairStrand_lower_C_01, bone2Name=dyn_hair_Mstr_r, swing1LimitDegrees=45, swing2LimitDegrees=45, twistLimitDegrees=1}
 {bone1Name=dyn_hairStrand_lower_C_02, bone2Name=dyn_hairStrand_lower_C_01, swing1LimitDegrees=45, swing2LimitDegrees=45, twistLimitDegrees=1}
 {bone1Name=dyn_hairStrand_lower_D_01, bone2Name=dyn_hair_Mstr_r, swing1LimitDegrees=45, swing2LimitDegrees=45, twistLimitDegrees=1}
 {bone1Name=dyn_hairStrand_lower_D_02, bone2Name=dyn_hairStrand_lower_D_01, swing1LimitDegrees=45, swing2LimitDegrees=45, twistLimitDegrees=1}
 {bone1Name=dyn_hairStrand_lower_D_03, bone2Name=dyn_hairStrand_lower_D_02, swing1LimitDegrees=45, swing2LimitDegrees=45, twistLimitDegrees=1}
 {bone1Name=dyn_hairStrand_lower_E_01, bone2Name=dyn_hair_Mstr_r, swing1LimitDegrees=45, swing2LimitDegrees=45, twistLimitDegrees=1}
 {bone1Name=dyn_hairStrand_lower_E_02, bone2Name=dyn_hairStrand_lower_E_01, swing1LimitDegrees=45, swing2LimitDegrees=45, twistLimitDegrees=1}
 {bone1Name=dyn_hairStrand_lower_F_01, bone2Name=dyn_hair_Mstr_r, swing1LimitDegrees=45, swing2LimitDegrees=45, twistLimitDegrees=1}
 {bone1Name=dyn_hairStrand_lower_F_02, bone2Name=dyn_hairStrand_lower_F_01, swing1LimitDegrees=45, swing2LimitDegrees=45, twistLimitDegrees=1}
 {bone1Name=dyn_hairStrand_lower_H, bone2Name=dyn_hair_Mstr_mid, swing1LimitDegrees=45, swing2LimitDegrees=45, twistLimitDegrees=1}
 {bone1Name=dyn_hairStrand_lower_I_01, bone2Name=dyn_hair_Mstr_mid, swing1LimitDegrees=45, swing2LimitDegrees=45, twistLimitDegrees=1}
 {bone1Name=dyn_hairStrand_lower_I_02, bone2Name=dyn_hairStrand_lower_I_01, swing1LimitDegrees=45, swing2LimitDegrees=45, twistLimitDegrees=1}
 {bone1Name=dyn_hairStrand_lower_I_03, bone2Name=dyn_hairStrand_lower_I_01, swing1LimitDegrees=45, swing2LimitDegrees=45, twistLimitDegrees=1}
 {bone1Name=dyn_hairStrand_upper_E_01, bone2Name=head, swing1LimitDegrees=45, swing2LimitDegrees=45, twistLimitDegrees=1}
 {bone1Name=dyn_hairStrand_upper_E_02, bone2Name=dyn_hairStrand_upper_E_01, swing1LimitDegrees=45, swing2LimitDegrees=45, twistLimitDegrees=1}
 {bone1Name=dyn_hairStrand_upper_E_03, bone2Name=dyn_hairStrand_upper_E_02, swing1LimitDegrees=45, swing2LimitDegrees=45, twistLimitDegrees=1}
 {bone1Name=dyn_hairStrand_upper_F_01, bone2Name=head, swing1LimitDegrees=45, swing2LimitDegrees=45, twistLimitDegrees=1}
 {bone1Name=dyn_hairStrand_upper_F_02, bone2Name=dyn_hairStrand_upper_F_01, swing1LimitDegrees=45, swing2LimitDegrees=45, twistLimitDegrees=1}
 {bone1Name=dyn_hairStrand_upper_I_01, bone2Name=head, swing1LimitDegrees=45, swing2LimitDegrees=45, twistLimitDegrees=1}
 {bone1Name=dyn_hairStrand_upper_H_01, bone2Name=dyn_hair_Mstr_l, swing1LimitDegrees=45, swing2LimitDegrees=45, twistLimitDegrees=1}
 {bone1Name=dyn_hairStrand_upper_D_01, bone2Name=dyn_hair_Mstr_r, swing1LimitDegrees=45, swing2LimitDegrees=45, twistLimitDegrees=1}
 {bone1Name=dyn_hairStrand_upper_D_02, bone2Name=dyn_hairStrand_upper_D_01, swing1LimitDegrees=45, swing2LimitDegrees=45, twistLimitDegrees=1}
 {bone1Name=dyn_hairStrand_upper_A_01, bone2Name=head, swing1LimitDegrees=45, swing2LimitDegrees=45, twistLimitDegrees=1}
 {bone1Name=dyn_hairStrand_upper_A_02, bone2Name=dyn_hairStrand_upper_A_01, swing1LimitDegrees=45, swing2LimitDegrees=45, twistLimitDegrees=1}
 {bone1Name=dyn_hairStrand_upper_B_01, bone2Name=head, swing1LimitDegrees=45, swing2LimitDegrees=45, twistLimitDegrees=1}
 {bone1Name=dyn_hairStrand_upper_B_02, bone2Name=dyn_hairStrand_upper_B_01, swing1LimitDegrees=45, swing2LimitDegrees=45, twistLimitDegrees=1}
 {bone1Name=dyn_hairStrand_upper_J_01, bone2Name=head, swing1LimitDegrees=45, swing2LimitDegrees=45, twistLimitDegrees=1}
 {bone1Name=dyn_hairStrand_upper_K_01, bone2Name=head, swing1LimitDegrees=45, swing2LimitDegrees=45, twistLimitDegrees=1}
]
