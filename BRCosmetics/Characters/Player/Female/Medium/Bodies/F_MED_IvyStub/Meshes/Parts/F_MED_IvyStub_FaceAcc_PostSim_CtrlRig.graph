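# Post-sim Control Rig for the IvyStub female face accessory that adds procedural waving to upper and lower hair strand bones.
# Sequenced ChainHarmonicsPerItem units drive the strands using only waves, with pendulum and reach disabled on those shown.
# Parent constraints blend selected lower strand bones with their neighbors.
# Skydiving, dive mode and interpolated custom emote alpha set shared WaveSpeed variables; non-dive skydiving also rotates one upper strand.
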ControlRigBlueprint /BRCosmetics/Characters/Player/Female/Medium/Bodies/F_MED_IvyStub/Meshes/Parts/F_MED_IvyStub_FaceAcc_PostSim_CtrlRig
graphs [
 RigVMFunctionLibrary {links [], nodes []}
 RigVMFunction_Sequence_1_ContainedGraph {nodes [
   n1 = RigVMFunctionEntryNode()
   RigVMFunction_Sequence = &RigVMFunction_Sequence(ExecuteContext=@n1.ExecuteContext)
   n2 = &RigVMFunction_Sequence(ExecuteContext=@RigVMFunction_Sequence.B)
   n3 = &RigVMFunction_Sequence(ExecuteContext=@n2.B)
   n4 = &RigVMFunction_Sequence(ExecuteContext=@n3.B)
   n5 = &RigVMFunction_Sequence(ExecuteContext=@n4.B)
   RigVMFunctionReturnNode(A=@RigVMFunction_Sequence.A, B=@n2.A, C=@n3.A, D=@n4.A, E=@n5.A, F=@n5.B)
  ]}
 RigVMModel {nodes [
   RigUnit_BeginExecution = &RigUnit_BeginExecution()
   RigVMFunction_Sequence = RigVMAggregateNode(ExecuteContext=@RigUnit_BeginExecution.ExecutePin)
   /Script/ControlRig.RigUnit_ChainHarmonicsPerItem(
    ChainRoot=(Type=Bone,Name="dyn_hairStrand_lower_I_01")
    DrawWorldOffset=(Rotation=(X=0,Y=0,Z=0,W=1),Translation=(X=0,Y=0,Z=0),Scale3D=(X=1,Y=1,Z=1))
    Pendulum=(bEnabled=false,PendulumStiffness=50,PendulumGravity=(X=-900,Y=-500,Z=0),PendulumBlend=1,PendulumDrag=20,PendulumMinimum=0,PendulumMaximum=0.8,PendulumEase=SineEaseInOut,UnwindAxis=(X=0,Y=0,Z=0),UnwindMinimum=0,UnwindMaximum=0)
    Reach=(bEnabled=false,ReachTarget=(X=0,Y=0,Z=0),ReachAxis=(X=1,Y=1,Z=1),ReachMinimum=0,ReachMaximum=0,ReachEase=SineEaseInOut)
    Speed=(X=0,Y=3,Z=0)
    Wave=(bEnabled=True,WaveFrequency=(X=0,Y=10,Z=0),WaveAmplitude=(X=0,Y=-2,Z=0),WaveOffset=(X=0,Y=1.5,Z=0),WaveNoise=(X=0,Y=0,Z=0),WaveMinimum=0.5,WaveMaximum=1,WaveEase=SineEaseInOut)
    WaveCurve="(EditorCurveData=(Keys=((InterpMode=RCIM_Linear,TangentMode=RCTM_Auto,TangentWeightMode=RCTWM_WeightedNone,Time=0,Value=0.2,ArriveTangent=0,ArriveTangentWeight=0,LeaveTangent=0,LeaveTangentWeight=0),(InterpMode=RCIM_Linear,TangentMode=RCTM_Auto,TangentWeig ...(truncated)"
    bDrawDebug=False
    ExecutePin=@RigVMFunction_Sequence.B
    Speed.Y=RigVMVariableNode() {variable=WaveSpeed 3}.Value
   ) {is=struct RigVMUnitNode}
   > &RigUnit_ChainHarmonicsPerItem(
    ChainRoot=(Type=Bone,Name="dyn_hairStrand_lower_I_02")
    DrawWorldOffset=(Rotation=(X=0,Y=0,Z=0,W=1),Translation=(X=0,Y=0,Z=0),Scale3D=(X=1,Y=1,Z=1))
    Pendulum=(bEnabled=false,PendulumStiffness=50,PendulumGravity=(X=-900,Y=-500,Z=0),PendulumBlend=1,PendulumDrag=0.98,PendulumMinimum=0,PendulumMaximum=0.9,PendulumEase=SineEaseInOut,UnwindAxis=(X=0,Y=0,Z=0),UnwindMinimum=0,UnwindMaximum=0)
    Reach=(bEnabled=false,ReachTarget=(X=0,Y=0,Z=0),ReachAxis=(X=1,Y=1,Z=1),ReachMinimum=0,ReachMaximum=0,ReachEase=SineEaseInOut)
    Speed=(X=2,Y=0,Z=2)
    Wave=(bEnabled=True,WaveFrequency=(X=1,Y=0,Z=1),WaveAmplitude=(X=1,Y=0,Z=1),WaveOffset=(X=2,Y=0,Z=2),WaveNoise=(X=0,Y=0,Z=0),WaveMinimum=-1,WaveMaximum=1,WaveEase=SineEaseInOut)
    WaveCurve="(EditorCurveData=(Keys=((InterpMode=RCIM_Linear,TangentMode=RCTM_Auto,TangentWeightMode=RCTWM_WeightedNone,Time=0,Value=0.2,ArriveTangent=0,ArriveTangentWeight=0,LeaveTangent=0,LeaveTangentWeight=0),(InterpMode=RCIM_Linear,TangentMode=RCTM_Auto,TangentWeig ...(truncated)"
    bDrawDebug=False
   )
   > &RigUnit_ChainHarmonicsPerItem(
    ChainRoot=(Type=Bone,Name="dyn_hairStrand_lower_J_01")
    DrawWorldOffset=(Rotation=(X=0,Y=0,Z=0,W=1),Translation=(X=0,Y=0,Z=0),Scale3D=(X=1,Y=1,Z=1))
    Pendulum=(bEnabled=false,PendulumStiffness=50,PendulumGravity=(X=-900,Y=-500,Z=0),PendulumBlend=1,PendulumDrag=20,PendulumMinimum=0,PendulumMaximum=0.8,PendulumEase=SineEaseInOut,UnwindAxis=(X=0,Y=0,Z=0),UnwindMinimum=0,UnwindMaximum=0)
    Reach=(bEnabled=false,ReachTarget=(X=0,Y=0,Z=0),ReachAxis=(X=1,Y=1,Z=1),ReachMinimum=0,ReachMaximum=0,ReachEase=SineEaseInOut)
    Speed=(X=0,Y=2,Z=0)
    Wave=(bEnabled=True,WaveFrequency=(X=0,Y=10,Z=0),WaveAmplitude=(X=0,Y=-3,Z=0),WaveOffset=(X=0,Y=1.5,Z=0),WaveNoise=(X=0,Y=0,Z=0),WaveMinimum=0.1,WaveMaximum=1,WaveEase=SineEaseInOut)
    WaveCurve="(EditorCurveData=(Keys=((InterpMode=RCIM_Linear,TangentMode=RCTM_Auto,TangentWeightMode=RCTWM_WeightedNone,Time=0,Value=0.2,ArriveTangent=0,ArriveTangentWeight=0,LeaveTangent=0,LeaveTangentWeight=0),(InterpMode=RCIM_Linear,TangentMode=RCTM_Auto,TangentWeig ...(truncated)"
    bDrawDebug=False
    Speed.Y=RigVMVariableNode() {variable=WaveSpeed 2}.Value
   )
   > &RigUnit_ChainHarmonicsPerItem(
    ChainRoot=(Type=Bone,Name="dyn_hairStrand_lower_J_02")
    DrawWorldOffset=(Rotation=(X=0,Y=0,Z=0,W=1),Translation=(X=0,Y=0,Z=0),Scale3D=(X=1,Y=1,Z=1))
    Pendulum=(bEnabled=false,PendulumStiffness=50,PendulumGravity=(X=-900,Y=-500,Z=0),PendulumBlend=1,PendulumDrag=0.98,PendulumMinimum=0,PendulumMaximum=0.9,PendulumEase=SineEaseInOut,UnwindAxis=(X=0,Y=0,Z=0),UnwindMinimum=0,UnwindMaximum=0)
    Reach=(bEnabled=false,ReachTarget=(X=0,Y=0,Z=0),ReachAxis=(X=1,Y=1,Z=1),ReachMinimum=0,ReachMaximum=0,ReachEase=SineEaseInOut)
    Speed=(X=2,Y=0,Z=2)
    Wave=(bEnabled=True,WaveFrequency=(X=1,Y=0,Z=1),WaveAmplitude=(X=1,Y=0,Z=1),WaveOffset=(X=0,Y=0,Z=0),WaveNoise=(X=0,Y=0,Z=0),WaveMinimum=-1,WaveMaximum=1,WaveEase=SineEaseInOut)
    WaveCurve="(EditorCurveData=(Keys=((InterpMode=RCIM_Linear,TangentMode=RCTM_Auto,TangentWeightMode=RCTWM_WeightedNone,Time=0,Value=0.2,ArriveTangent=0,ArriveTangentWeight=0,LeaveTangent=0,LeaveTangentWeight=0),(InterpMode=RCIM_Linear,TangentMode=RCTM_Auto,TangentWeig ...(truncated)"
    bDrawDebug=False
   )
   > &RigUnit_ChainHarmonicsPerItem(
    ChainRoot=(Type=Bone,Name="dyn_hairStrand_lower_K_01")
    DrawWorldOffset=(Rotation=(X=0,Y=0,Z=0,W=1),Translation=(X=0,Y=0,Z=0),Scale3D=(X=1,Y=1,Z=1))
    Pendulum=(bEnabled=false,PendulumStiffness=50,PendulumGravity=(X=-900,Y=-500,Z=0),PendulumBlend=1,PendulumDrag=20,PendulumMinimum=0,PendulumMaximum=0.8,PendulumEase=SineEaseInOut,UnwindAxis=(X=0,Y=0,Z=0),UnwindMinimum=0,UnwindMaximum=0)
    Reach=(bEnabled=false,ReachTarget=(X=0,Y=0,Z=0),ReachAxis=(X=1,Y=1,Z=1),ReachMinimum=0,ReachMaximum=0,ReachEase=SineEaseInOut)
    Speed=(X=0,Y=2,Z=0)
    Wave=(bEnabled=True,WaveFrequency=(X=0,Y=10,Z=0),WaveAmplitude=(X=0,Y=-2,Z=0),WaveOffset=(X=0,Y=2,Z=0),WaveNoise=(X=0,Y=0,Z=0),WaveMinimum=0.1,WaveMaximum=1,WaveEase=SineEaseInOut)
    WaveCurve="(EditorCurveData=(Keys=((InterpMode=RCIM_Linear,TangentMode=RCTM_Auto,TangentWeightMode=RCTWM_WeightedNone,Time=0,Value=0.2,ArriveTangent=0,ArriveTangentWeight=0,LeaveTangent=0,LeaveTangentWeight=0),(InterpMode=RCIM_Linear,TangentMode=RCTM_Auto,TangentWeig ...(truncated)"
    bDrawDebug=False
    Speed.Y=RigVMVariableNode() {variable=WaveSpeed 2}.Value
   )
   > &RigUnit_ChainHarmonicsPerItem(
    ChainRoot=(Type=Bone,Name="dyn_hairStrand_lower_K_02")
    DrawWorldOffset=(Rotation=(X=0,Y=0,Z=0,W=1),Translation=(X=0,Y=0,Z=0),Scale3D=(X=1,Y=1,Z=1))
    Pendulum=(bEnabled=false,PendulumStiffness=50,PendulumGravity=(X=-900,Y=-500,Z=0),PendulumBlend=1,PendulumDrag=0.98,PendulumMinimum=0,PendulumMaximum=0.9,PendulumEase=SineEaseInOut,UnwindAxis=(X=0,Y=0,Z=0),UnwindMinimum=0,UnwindMaximum=0)
    Reach=(bEnabled=false,ReachTarget=(X=0,Y=0,Z=0),ReachAxis=(X=1,Y=1,Z=1),ReachMinimum=0,ReachMaximum=0,ReachEase=SineEaseInOut)
    Speed=(X=2,Y=0,Z=2)
    Wave=(bEnabled=True,WaveFrequency=(X=1,Y=0,Z=1),WaveAmplitude=(X=1,Y=0,Z=1),WaveOffset=(X=2,Y=0,Z=2),WaveNoise=(X=0,Y=0,Z=0),WaveMinimum=-1,WaveMaximum=1,WaveEase=SineEaseInOut)
    WaveCurve="(EditorCurveData=(Keys=((InterpMode=RCIM_Linear,TangentMode=RCTM_Auto,TangentWeightMode=RCTWM_WeightedNone,Time=0,Value=0.2,ArriveTangent=0,ArriveTangentWeight=0,LeaveTangent=0,LeaveTangentWeight=0),(InterpMode=RCIM_Linear,TangentMode=RCTM_Auto,TangentWeig ...(truncated)"
    bDrawDebug=False
   )
   > &RigUnit_ChainHarmonicsPerItem(
    ChainRoot=(Type=Bone,Name="dyn_hairStrand_lower_L_01")
    DrawWorldOffset=(Rotation=(X=0,Y=0,Z=0,W=1),Translation=(X=0,Y=0,Z=0),Scale3D=(X=1,Y=1,Z=1))
    Pendulum=(bEnabled=false,PendulumStiffness=50,PendulumGravity=(X=-900,Y=-500,Z=0),PendulumBlend=1,PendulumDrag=20,PendulumMinimum=0,PendulumMaximum=0.8,PendulumEase=SineEaseInOut,UnwindAxis=(X=0,Y=0,Z=0),UnwindMinimum=0,UnwindMaximum=0)
    Reach=(bEnabled=false,ReachTarget=(X=0,Y=0,Z=0),ReachAxis=(X=1,Y=1,Z=1),ReachMinimum=0,ReachMaximum=0,ReachEase=SineEaseInOut)
    Speed=(X=0,Y=3,Z=0)
    Wave=(bEnabled=True,WaveFrequency=(X=0,Y=10,Z=0),WaveAmplitude=(X=0,Y=-1,Z=0),WaveOffset=(X=0,Y=2.3,Z=0),WaveNoise=(X=0,Y=0,Z=0),WaveMinimum=0.1,WaveMaximum=1,WaveEase=SineEaseInOut)
    WaveCurve="(EditorCurveData=(Keys=((InterpMode=RCIM_Linear,TangentMode=RCTM_Auto,TangentWeightMode=RCTWM_WeightedNone,Time=0,Value=0.2,ArriveTangent=0,ArriveTangentWeight=0,LeaveTangent=0,LeaveTangentWeight=0),(InterpMode=RCIM_Linear,TangentMode=RCTM_Auto,TangentWeig ...(truncated)"
    bDrawDebug=False
    Speed.Y=RigVMVariableNode() {variable=WaveSpeed 3}.Value
   )
   > &RigUnit_ChainHarmonicsPerItem(
    ChainRoot=(Type=Bone,Name="dyn_hairStrand_lower_L_02")
    DrawWorldOffset=(Rotation=(X=0,Y=0,Z=0,W=1),Translation=(X=0,Y=0,Z=0),Scale3D=(X=1,Y=1,Z=1))
    Pendulum=(bEnabled=false,PendulumStiffness=50,PendulumGravity=(X=-900,Y=-500,Z=0),PendulumBlend=1,PendulumDrag=0.98,PendulumMinimum=0,PendulumMaximum=0.9,PendulumEase=SineEaseInOut,UnwindAxis=(X=0,Y=0,Z=0),UnwindMinimum=0,UnwindMaximum=0)
    Reach=(bEnabled=false,ReachTarget=(X=0,Y=0,Z=0),ReachAxis=(X=1,Y=1,Z=1),ReachMinimum=0,ReachMaximum=0,ReachEase=SineEaseInOut)
    Speed=(X=2,Y=0,Z=2)
    Wave=(bEnabled=True,WaveFrequency=(X=1,Y=0,Z=1),WaveAmplitude=(X=1,Y=0,Z=1),WaveOffset=(X=0,Y=0,Z=0),WaveNoise=(X=0,Y=0,Z=0),WaveMinimum=-1,WaveMaximum=1,WaveEase=SineEaseInOut)
    WaveCurve="(EditorCurveData=(Keys=((InterpMode=RCIM_Linear,TangentMode=RCTM_Auto,TangentWeightMode=RCTWM_WeightedNone,Time=0,Value=0.2,ArriveTangent=0,ArriveTangentWeight=0,LeaveTangent=0,LeaveTangentWeight=0),(InterpMode=RCIM_Linear,TangentMode=RCTM_Auto,TangentWeig ...(truncated)"
    bDrawDebug=False
   )
   Multiply = &RigVMFunction_MathDoubleMul(B=-1, A=RigVMVariableNode() {variable=WaveSpeed 3}.Value)
   > &RigUnit_ChainHarmonicsPerItem(
    ChainRoot=(Type=Bone,Name="dyn_hairStrand_lower_M_01")
    DrawWorldOffset=(Rotation=(X=0,Y=0,Z=0,W=1),Translation=(X=0,Y=0,Z=0),Scale3D=(X=1,Y=1,Z=1))
    Pendulum=(bEnabled=false,PendulumStiffness=50,PendulumGravity=(X=-900,Y=-500,Z=0),PendulumBlend=1,PendulumDrag=20,PendulumMinimum=0,PendulumMaximum=0.8,PendulumEase=SineEaseInOut,UnwindAxis=(X=0,Y=0,Z=0),UnwindMinimum=0,UnwindMaximum=0)
    Reach=(bEnabled=false,ReachTarget=(X=0,Y=0,Z=0),ReachAxis=(X=1,Y=1,Z=1),ReachMinimum=0,ReachMaximum=0,ReachEase=SineEaseInOut)
    Speed=(X=0,Y=-3,Z=0)
    Wave=(bEnabled=True,WaveFrequency=(X=0,Y=5,Z=0),WaveAmplitude=(X=0,Y=-1.7,Z=0),WaveOffset=(X=0,Y=2,Z=0),WaveNoise=(X=0,Y=0,Z=0),WaveMinimum=0.1,WaveMaximum=1,WaveEase=SineEaseInOut)
    WaveCurve="(EditorCurveData=(Keys=((InterpMode=RCIM_Linear,TangentMode=RCTM_Auto,TangentWeightMode=RCTWM_WeightedNone,Time=0,Value=0.2,ArriveTangent=0,ArriveTangentWeight=0,LeaveTangent=0,LeaveTangentWeight=0),(InterpMode=RCIM_Linear,TangentMode=RCTM_Auto,TangentWeig ...(truncated)"
    bDrawDebug=False
    Speed.Y=@Multiply.Result
   )
   > &RigUnit_ChainHarmonicsPerItem(
    ChainRoot=(Type=Bone,Name="dyn_hairStrand_lower_N_01")
    DrawWorldOffset=(Rotation=(X=0,Y=0,Z=0,W=1),Translation=(X=0,Y=0,Z=0),Scale3D=(X=1,Y=1,Z=1))
    Pendulum=(bEnabled=false,PendulumStiffness=50,PendulumGravity=(X=-900,Y=-500,Z=0),PendulumBlend=1,PendulumDrag=20,PendulumMinimum=0,PendulumMaximum=0.8,PendulumEase=SineEaseInOut,UnwindAxis=(X=0,Y=0,Z=0),UnwindMinimum=0,UnwindMaximum=0)
    Reach=(bEnabled=false,ReachTarget=(X=0,Y=0,Z=0),ReachAxis=(X=1,Y=1,Z=1),ReachMinimum=0,ReachMaximum=0,ReachEase=SineEaseInOut)
    Speed=(X=0,Y=2,Z=0)
    Wave=(bEnabled=True,WaveFrequency=(X=0,Y=10,Z=0),WaveAmplitude=(X=0,Y=-2,Z=0),WaveOffset=(X=0,Y=1.5,Z=0),WaveNoise=(X=0,Y=0,Z=0),WaveMinimum=0.1,WaveMaximum=1,WaveEase=SineEaseInOut)
    WaveCurve="(EditorCurveData=(Keys=((InterpMode=RCIM_Linear,TangentMode=RCTM_Auto,TangentWeightMode=RCTWM_WeightedNone,Time=0,Value=0.2,ArriveTangent=0,ArriveTangentWeight=0,LeaveTangent=0,LeaveTangentWeight=0),(InterpMode=RCIM_Linear,TangentMode=RCTM_Auto,TangentWeig ...(truncated)"
    bDrawDebug=False
    Speed.Y=RigVMVariableNode() {variable=WaveSpeed 2}.Value
   )
   > &RigUnit_ChainHarmonicsPerItem(
    ChainRoot=(Type=Bone,Name="dyn_hairStrand_lower_N_02")
    DrawWorldOffset=(Rotation=(X=0,Y=0,Z=0,W=1),Translation=(X=0,Y=0,Z=0),Scale3D=(X=1,Y=1,Z=1))
    Pendulum=(bEnabled=false,PendulumStiffness=50,PendulumGravity=(X=-900,Y=-500,Z=0),PendulumBlend=1,PendulumDrag=0.98,PendulumMinimum=0,PendulumMaximum=0.9,PendulumEase=SineEaseInOut,UnwindAxis=(X=0,Y=0,Z=0),UnwindMinimum=0,UnwindMaximum=0)
    Reach=(bEnabled=false,ReachTarget=(X=0,Y=0,Z=0),ReachAxis=(X=1,Y=1,Z=1),ReachMinimum=0,ReachMaximum=0,ReachEase=SineEaseInOut)
    Speed=(X=2,Y=0,Z=2)
    Wave=(bEnabled=True,WaveFrequency=(X=1,Y=0,Z=1),WaveAmplitude=(X=1,Y=0,Z=1),WaveOffset=(X=0,Y=0,Z=0),WaveNoise=(X=0,Y=0,Z=0),WaveMinimum=-1,WaveMaximum=1,WaveEase=SineEaseInOut)
    WaveCurve="(EditorCurveData=(Keys=((InterpMode=RCIM_Linear,TangentMode=RCTM_Auto,TangentWeightMode=RCTWM_WeightedNone,Time=0,Value=0.2,ArriveTangent=0,ArriveTangentWeight=0,LeaveTangent=0,LeaveTangentWeight=0),(InterpMode=RCIM_Linear,TangentMode=RCTM_Auto,TangentWeig ...(truncated)"
    bDrawDebug=False
   )
   &RigUnit_ParentConstraint(
    AdvancedSettings=(InterpolationType=Average,RotationOrderForFilter=XZY)
    Child=(Type=Bone,Name="dyn_hairStrand_lower_I_03")
    Parents=((Item=(Type=Bone,Name="dyn_hairStrand_lower_I_02"),Weight=1))
    Weight=0.5
    ExecutePin=@RigVMFunction_Sequence.E
   )
   > n1 = &RigUnit_ParentConstraint(
    AdvancedSettings=(InterpolationType=Average,RotationOrderForFilter=XZY)
    Child=(Type=Bone,Name="dyn_hairStrand_lower_H")
    Parents=((Item=(Type=Bone,Name="dyn_hairStrand_lower_G_02"),Weight=0.5),(Item=(Type=Bone,Name="dyn_hairStrand_lower_I_03"),Weight=0.5))
    Weight=1
   )
   RigVMFunction_ControlFlowBranch = &RigVMFunction_ControlFlowBranch(ExecuteContext=@n1.ExecutePin, Condition=RigVMVariableNode() {variable=Is Skydiving Dive Mode}.Value)
   n2 = &RigVMFunction_ControlFlowBranch(ExecuteContext=@RigVMFunction_ControlFlowBranch.False, Condition=RigVMVariableNode() {variable=Is Skydiving}.Value)
   n3 = &RigVMFunction_AlphaInterp(InterpSpeedDecreasing=2, InterpSpeedIncreasing=8, Scale=1, bClampResult=true, Value=RigVMVariableNode() {variable=Custom Emote Alpha}.Value)
   Greater = /Script/RigVM.RigVMFunction_MathDoubleGreater(B=0, A=@n3.Result) {is=struct RigVMUnitNode, template="Greater::Execute(in A,in B,out Result)"}
   n4 = &RigVMFunction_ControlFlowBranch(Condition=@Greater.Result, ExecuteContext=@n2.Completed)
   RigVMVariableNode(Value=4, ExecuteContext=@n4.False) {variable=WaveSpeed 4}
   > RigVMVariableNode(Value=3) {variable=WaveSpeed 3}
   > RigVMVariableNode(Value=2) {variable=WaveSpeed 2}
   n5 = &RigVMFunction_MathDoubleMul(B=10, A=@n3.Result)
   n6 = &RigVMFunction_AlphaInterp(ClampMax=40, ClampMin=4, InterpSpeedDecreasing=8, InterpSpeedIncreasing=8, Scale=1, bClampResult=true, Value=@n5.Result)
   RigVMVariableNode(Value=@n6.Result, ExecuteContext=@n4.True) {variable=WaveSpeed 4}
   n7 = &RigVMFunction_AlphaInterp(ClampMax=30, ClampMin=3, InterpSpeedDecreasing=8, InterpSpeedIncreasing=8, Scale=1, bClampResult=true, Value=@n5.Result)
   > RigVMVariableNode(Value=@n7.Result) {variable=WaveSpeed 3}
   n8 = &RigVMFunction_AlphaInterp(ClampMax=20, ClampMin=2, InterpSpeedDecreasing=8, InterpSpeedIncreasing=8, Scale=1, bClampResult=true, Value=@n5.Result)
   > RigVMVariableNode(Value=@n8.Result) {variable=WaveSpeed 2}
   RigVMVariableNode(Value=4, ExecuteContext=@n2.False) {variable=WaveSpeed 4}
   > RigVMVariableNode(Value=3) {variable=WaveSpeed 3}
   > RigVMVariableNode(Value=2) {variable=WaveSpeed 2}
   RigVMVariableNode(Value=15, ExecuteContext=@n2.True) {variable=WaveSpeed 4}
   > RigVMVariableNode(Value=12) {variable=WaveSpeed 3}
   > RigVMVariableNode(Value=9) {variable=WaveSpeed 2}
   > &RigUnit_ModifyTransforms(
    ItemToModify=((Item=(Type=Bone,Name="dyn_hairStrand_upper_E_01"),Transform=(Rotation=(X=0,Y=-0.1736,Z=0,W=-0.9848),Translation=(X=0,Y=0,Z=0),Scale3D=(X=1,Y=1,Z=1))))
    Weight=1
    WeightMinimum=0
   )
   RigVMVariableNode(Value=20, ExecuteContext=@RigVMFunction_ControlFlowBranch.True) {variable=WaveSpeed 4}
   > RigVMVariableNode(Value=20) {variable=WaveSpeed 3}
   > RigVMVariableNode(Value=20) {variable=WaveSpeed 2}
   &RigUnit_ChainHarmonicsPerItem(
    ChainRoot=(Type=Bone,Name="dyn_hairStrand_upper_A_01")
    DrawWorldOffset=(Rotation=(X=0,Y=0,Z=0,W=1),Translation=(X=0,Y=0,Z=0),Scale3D=(X=1,Y=1,Z=1))
    Pendulum=(bEnabled=false,PendulumStiffness=50,PendulumGravity=(X=-900,Y=-500,Z=0),PendulumBlend=1,PendulumDrag=20,PendulumMinimum=0,PendulumMaximum=0.8,PendulumEase=SineEaseInOut,UnwindAxis=(X=0,Y=0,Z=0),UnwindMinimum=0,UnwindMaximum=0)
    Reach=(bEnabled=false,ReachTarget=(X=0,Y=0,Z=0),ReachAxis=(X=1,Y=1,Z=1),ReachMinimum=0,ReachMaximum=0,ReachEase=SineEaseInOut)
    Speed=(X=0,Y=2,Z=0)
    Wave=(bEnabled=True,WaveFrequency=(X=0,Y=10,Z=0),WaveAmplitude=(X=0,Y=-2,Z=0),WaveOffset=(X=0,Y=2,Z=0),WaveNoise=(X=0,Y=0,Z=0),WaveMinimum=0.1,WaveMaximum=1,WaveEase=SineEaseIn)
    WaveCurve="(EditorCurveData=(Keys=((InterpMode=RCIM_Linear,TangentMode=RCTM_Auto,TangentWeightMode=RCTWM_WeightedNone,Time=0,Value=0.2,ArriveTangent=0,ArriveTangentWeight=0,LeaveTangent=0,LeaveTangentWeight=0),(InterpMode=RCIM_Linear,TangentMode=RCTM_Auto,TangentWeig ...(truncated)"
    bDrawDebug=false
    Speed.Y=RigVMVariableNode() {variable=WaveSpeed 2}.Value
    ExecutePin=@RigVMFunction_Sequence.C
   )
   > &RigUnit_ChainHarmonicsPerItem(
    ChainRoot=(Type=Bone,Name="dyn_hairStrand_upper_B_01")
    DrawWorldOffset=(Rotation=(X=0,Y=0,Z=0,W=1),Translation=(X=0,Y=0,Z=0),Scale3D=(X=1,Y=1,Z=1))
    Pendulum=(bEnabled=false,PendulumStiffness=50,PendulumGravity=(X=-900,Y=-500,Z=0),PendulumBlend=1,PendulumDrag=20,PendulumMinimum=0,PendulumMaximum=0.8,PendulumEase=SineEaseInOut,UnwindAxis=(X=0,Y=0,Z=0),UnwindMinimum=0,UnwindMaximum=0)
    Reach=(bEnabled=false,ReachTarget=(X=0,Y=0,Z=0),ReachAxis=(X=1,Y=1,Z=1),ReachMinimum=0,ReachMaximum=0,ReachEase=SineEaseInOut)
    Speed=(X=0,Y=2,Z=0)
    Wave=(bEnabled=True,WaveFrequency=(X=0,Y=10,Z=0),WaveAmplitude=(X=0,Y=-2,Z=0),WaveOffset=(X=0,Y=1.5,Z=0),WaveNoise=(X=0,Y=0,Z=0),WaveMinimum=0.1,WaveMaximum=1,WaveEase=SineEaseInOut)
    WaveCurve="(EditorCurveData=(Keys=((InterpMode=RCIM_Linear,TangentMode=RCTM_Auto,TangentWeightMode=RCTWM_WeightedNone,Time=0,Value=0.2,ArriveTangent=0,ArriveTangentWeight=0,LeaveTangent=0,LeaveTangentWeight=0),(InterpMode=RCIM_Linear,TangentMode=RCTM_Auto,TangentWeig ...(truncated)"
    bDrawDebug=False
    Speed.Y=RigVMVariableNode() {variable=WaveSpeed 2}.Value
   )
   > &RigUnit_ChainHarmonicsPerItem(
    ChainRoot=(Type=Bone,Name="dyn_hairStrand_upper_C_01")
    DrawWorldOffset=(Rotation=(X=0,Y=0,Z=0,W=1),Translation=(X=0,Y=0,Z=0),Scale3D=(X=1,Y=1,Z=1))
    Pendulum=(bEnabled=false,PendulumStiffness=50,PendulumGravity=(X=-900,Y=-500,Z=0),PendulumBlend=1,PendulumDrag=20,PendulumMinimum=0,PendulumMaximum=0.8,PendulumEase=SineEaseInOut,UnwindAxis=(X=0,Y=0,Z=0),UnwindMinimum=0,UnwindMaximum=0)
    Reach=(bEnabled=false,ReachTarget=(X=0,Y=0,Z=0),ReachAxis=(X=1,Y=1,Z=1),ReachMinimum=0,ReachMaximum=0,ReachEase=SineEaseInOut)
    Speed=(X=0,Y=2,Z=0)
    Wave=(bEnabled=True,WaveFrequency=(X=0,Y=10,Z=0),WaveAmplitude=(X=0,Y=-2,Z=0),WaveOffset=(X=0,Y=1,Z=0),WaveNoise=(X=0,Y=0,Z=0),WaveMinimum=0.1,WaveMaximum=1,WaveEase=SineEaseInOut)
    WaveCurve="(EditorCurveData=(Keys=((InterpMode=RCIM_Linear,TangentMode=RCTM_Auto,TangentWeightMode=RCTWM_WeightedNone,Time=0,Value=0.2,ArriveTangent=0,ArriveTangentWeight=0,LeaveTangent=0,LeaveTangentWeight=0),(InterpMode=RCIM_Linear,TangentMode=RCTM_Auto,TangentWeig ...(truncated)"
    bDrawDebug=False
    Speed.Y=RigVMVariableNode() {variable=WaveSpeed 2}.Value
   )
   > &RigUnit_ChainHarmonicsPerItem(
    ChainRoot=(Type=Bone,Name="dyn_hairStrand_upper_D_01")
    DrawWorldOffset=(Rotation=(X=0,Y=0,Z=0,W=1),Translation=(X=0,Y=0,Z=0),Scale3D=(X=1,Y=1,Z=1))
    Pendulum=(bEnabled=false,PendulumStiffness=50,PendulumGravity=(X=-900,Y=-500,Z=0),PendulumBlend=1,PendulumDrag=20,PendulumMinimum=0,PendulumMaximum=0.8,PendulumEase=SineEaseInOut,UnwindAxis=(X=0,Y=0,Z=0),UnwindMinimum=0,UnwindMaximum=0)
    Reach=(bEnabled=false,ReachTarget=(X=0,Y=0,Z=0),ReachAxis=(X=1,Y=1,Z=1),ReachMinimum=0,ReachMaximum=0,ReachEase=SineEaseInOut)
    Speed=(X=0,Y=3,Z=0)
    Wave=(bEnabled=True,WaveFrequency=(X=0,Y=10,Z=0),WaveAmplitude=(X=0,Y=-1,Z=0),WaveOffset=(X=0,Y=2.3,Z=0),WaveNoise=(X=0,Y=0,Z=0),WaveMinimum=0.1,WaveMaximum=1,WaveEase=SineEaseInOut)
    WaveCurve="(EditorCurveData=(Keys=((InterpMode=RCIM_Linear,TangentMode=RCTM_Auto,TangentWeightMode=RCTWM_WeightedNone,Time=0,Value=0.2,ArriveTangent=0,ArriveTangentWeight=0,LeaveTangent=0,LeaveTangentWeight=0),(InterpMode=RCIM_Linear,TangentMode=RCTM_Auto,TangentWeig ...(truncated)"
    bDrawDebug=False
    Speed.Y=RigVMVariableNode() {variable=WaveSpeed 3}.Value
   )
   > &RigUnit_ChainHarmonicsPerItem(
    ChainRoot=(Type=Bone,Name="dyn_hairStrand_upper_D_02")
    DrawWorldOffset=(Rotation=(X=0,Y=0,Z=0,W=1),Translation=(X=0,Y=0,Z=0),Scale3D=(X=1,Y=1,Z=1))
    Pendulum=(bEnabled=false,PendulumStiffness=50,PendulumGravity=(X=-900,Y=-500,Z=0),PendulumBlend=1,PendulumDrag=0.98,PendulumMinimum=0,PendulumMaximum=0.9,PendulumEase=SineEaseInOut,UnwindAxis=(X=0,Y=0,Z=0),UnwindMinimum=0,UnwindMaximum=0)
    Reach=(bEnabled=false,ReachTarget=(X=0,Y=0,Z=0),ReachAxis=(X=1,Y=1,Z=1),ReachMinimum=0,ReachMaximum=0,ReachEase=SineEaseInOut)
    Speed=(X=2,Y=0,Z=2)
    Wave=(bEnabled=True,WaveFrequency=(X=1,Y=0,Z=1),WaveAmplitude=(X=1,Y=0,Z=1),WaveOffset=(X=0,Y=0,Z=0),WaveNoise=(X=0,Y=0,Z=0),WaveMinimum=-1,WaveMaximum=1,WaveEase=SineEaseInOut)
    WaveCurve="(EditorCurveData=(Keys=((InterpMode=RCIM_Linear,TangentMode=RCTM_Auto,TangentWeightMode=RCTWM_WeightedNone,Time=0,Value=0.2,ArriveTangent=0,ArriveTangentWeight=0,LeaveTangent=0,LeaveTangentWeight=0),(InterpMode=RCIM_Linear,TangentMode=RCTM_Auto,TangentWeig ...(truncated)"
    bDrawDebug=False
   )
   > &RigUnit_ChainHarmonicsPerItem(
    ChainRoot=(Type=Bone,Name="dyn_hairStrand_upper_E_01")
    DrawWorldOffset=(Rotation=(X=0,Y=0,Z=0,W=1),Translation=(X=0,Y=0,Z=0),Scale3D=(X=1,Y=1,Z=1))
    Pendulum=(bEnabled=false,PendulumStiffness=50,PendulumGravity=(X=-900,Y=-500,Z=0),PendulumBlend=1,PendulumDrag=20,PendulumMinimum=0,PendulumMaximum=0.8,PendulumEase=SineEaseInOut,UnwindAxis=(X=0,Y=0,Z=0),UnwindMinimum=0,UnwindMaximum=0)
    Reach=(bEnabled=false,ReachTarget=(X=0,Y=0,Z=0),ReachAxis=(X=1,Y=1,Z=1),ReachMinimum=0,ReachMaximum=0,ReachEase=SineEaseInOut)
    Speed=(X=0,Y=3,Z=0)
    Wave=(bEnabled=True,WaveFrequency=(X=0,Y=10,Z=0),WaveAmplitude=(X=0,Y=-1,Z=0),WaveOffset=(X=0,Y=2,Z=0),WaveNoise=(X=0,Y=0,Z=0),WaveMinimum=0.1,WaveMaximum=1,WaveEase=SineEaseInOut)
    WaveCurve="(EditorCurveData=(Keys=((InterpMode=RCIM_Linear,TangentMode=RCTM_Auto,TangentWeightMode=RCTWM_WeightedNone,Time=0,Value=0.2,ArriveTangent=0,ArriveTangentWeight=0,LeaveTangent=0,LeaveTangentWeight=0),(InterpMode=RCIM_Linear,TangentMode=RCTM_Auto,TangentWeig ...(truncated)"
    bDrawDebug=False
    Speed.Y=RigVMVariableNode() {variable=WaveSpeed 3}.Value
   )
   > &RigUnit_ChainHarmonicsPerItem(
    ChainRoot=(Type=Bone,Name="dyn_hairStrand_upper_E_02")
    DrawWorldOffset=(Rotation=(X=0,Y=0,Z=0,W=1),Translation=(X=0,Y=0,Z=0),Scale3D=(X=1,Y=1,Z=1))
    Pendulum=(bEnabled=false,PendulumStiffness=50,PendulumGravity=(X=-900,Y=-500,Z=0),PendulumBlend=1,PendulumDrag=0.98,PendulumMinimum=0,PendulumMaximum=0.9,PendulumEase=SineEaseInOut,UnwindAxis=(X=0,Y=0,Z=0),UnwindMinimum=0,UnwindMaximum=0)
    Reach=(bEnabled=false,ReachTarget=(X=0,Y=0,Z=0),ReachAxis=(X=1,Y=1,Z=1),ReachMinimum=0,ReachMaximum=0,ReachEase=SineEaseInOut)
    Speed=(X=2,Y=0,Z=2)
    Wave=(bEnabled=True,WaveFrequency=(X=1,Y=0,Z=1),WaveAmplitude=(X=1,Y=0,Z=1),WaveOffset=(X=2,Y=0,Z=2),WaveNoise=(X=0,Y=0,Z=0),WaveMinimum=-1,WaveMaximum=1,WaveEase=SineEaseInOut)
    WaveCurve="(EditorCurveData=(Keys=((InterpMode=RCIM_Linear,TangentMode=RCTM_Auto,TangentWeightMode=RCTWM_WeightedNone,Time=0,Value=0.2,ArriveTangent=0,ArriveTangentWeight=0,LeaveTangent=0,LeaveTangentWeight=0),(InterpMode=RCIM_Linear,TangentMode=RCTM_Auto,TangentWeig ...(truncated)"
    bDrawDebug=False
   )
   > &RigUnit_ChainHarmonicsPerItem(
    ChainRoot=(Type=Bone,Name="dyn_hairStrand_upper_F_01")
    DrawWorldOffset=(Rotation=(X=0,Y=0,Z=0,W=1),Translation=(X=0,Y=0,Z=0),Scale3D=(X=1,Y=1,Z=1))
    Pendulum=(bEnabled=false,PendulumStiffness=50,PendulumGravity=(X=-900,Y=-500,Z=0),PendulumBlend=1,PendulumDrag=20,PendulumMinimum=0,PendulumMaximum=0.8,PendulumEase=SineEaseInOut,UnwindAxis=(X=0,Y=0,Z=0),UnwindMinimum=0,UnwindMaximum=0)
    Reach=(bEnabled=false,ReachTarget=(X=0,Y=0,Z=0),ReachAxis=(X=1,Y=1,Z=1),ReachMinimum=0,ReachMaximum=0,ReachEase=SineEaseInOut)
    Speed=(X=0,Y=3,Z=0)
    Wave=(bEnabled=True,WaveFrequency=(X=0,Y=10,Z=0),WaveAmplitude=(X=0,Y=-2,Z=0),WaveOffset=(X=0,Y=1.5,Z=0),WaveNoise=(X=0,Y=0,Z=0),WaveMinimum=0.1,WaveMaximum=1,WaveEase=SineEaseInOut)
    WaveCurve="(EditorCurveData=(Keys=((InterpMode=RCIM_Linear,TangentMode=RCTM_Auto,TangentWeightMode=RCTWM_WeightedNone,Time=0,Value=0.2,ArriveTangent=0,ArriveTangentWeight=0,LeaveTangent=0,LeaveTangentWeight=0),(InterpMode=RCIM_Linear,TangentMode=RCTM_Auto,TangentWeig ...(truncated)"
    bDrawDebug=False
    Speed.Y=RigVMVariableNode() {variable=WaveSpeed 3}.Value
   )
   > &RigUnit_ChainHarmonicsPerItem(
    ChainRoot=(Type=Bone,Name="dyn_hairStrand_upper_F_02")
    DrawWorldOffset=(Rotation=(X=0,Y=0,Z=0,W=1),Translation=(X=0,Y=0,Z=0),Scale3D=(X=1,Y=1,Z=1))
    Pendulum=(bEnabled=false,PendulumStiffness=50,PendulumGravity=(X=-900,Y=-500,Z=0),PendulumBlend=1,PendulumDrag=0.98,PendulumMinimum=0,PendulumMaximum=0.9,PendulumEase=SineEaseInOut,UnwindAxis=(X=0,Y=0,Z=0),UnwindMinimum=0,UnwindMaximum=0)
    Reach=(bEnabled=false,ReachTarget=(X=0,Y=0,Z=0),ReachAxis=(X=1,Y=1,Z=1),ReachMinimum=0,ReachMaximum=0,ReachEase=SineEaseInOut)
    Speed=(X=2,Y=0,Z=2)
    Wave=(bEnabled=True,WaveFrequency=(X=1,Y=0,Z=1),WaveAmplitude=(X=1,Y=0,Z=1),WaveOffset=(X=0,Y=0,Z=0),WaveNoise=(X=0,Y=0,Z=0),WaveMinimum=-1,WaveMaximum=1,WaveEase=SineEaseInOut)
    WaveCurve="(EditorCurveData=(Keys=((InterpMode=RCIM_Linear,TangentMode=RCTM_Auto,TangentWeightMode=RCTWM_WeightedNone,Time=0,Value=0.2,ArriveTangent=0,ArriveTangentWeight=0,LeaveTangent=0,LeaveTangentWeight=0),(InterpMode=RCIM_Linear,TangentMode=RCTM_Auto,TangentWeig ...(truncated)"
    bDrawDebug=False
   )
   > &RigUnit_ChainHarmonicsPerItem(
    ChainRoot=(Type=Bone,Name="dyn_hairStrand_upper_G_01")
    DrawWorldOffset=(Rotation=(X=0,Y=0,Z=0,W=1),Translation=(X=0,Y=0,Z=0),Scale3D=(X=1,Y=1,Z=1))
    Pendulum=(bEnabled=false,PendulumStiffness=50,PendulumGravity=(X=-900,Y=-500,Z=0),PendulumBlend=1,PendulumDrag=20,PendulumMinimum=0,PendulumMaximum=0.8,PendulumEase=SineEaseInOut,UnwindAxis=(X=0,Y=0,Z=0),UnwindMinimum=0,UnwindMaximum=0)
    Reach=(bEnabled=false,ReachTarget=(X=0,Y=0,Z=0),ReachAxis=(X=1,Y=1,Z=1),ReachMinimum=0,ReachMaximum=0,ReachEase=SineEaseInOut)
    Speed=(X=0,Y=2,Z=0)
    Wave=(bEnabled=True,WaveFrequency=(X=0,Y=10,Z=0),WaveAmplitude=(X=0,Y=-2.2,Z=0),WaveOffset=(X=0,Y=2,Z=0),WaveNoise=(X=0,Y=0,Z=0),WaveMinimum=0.1,WaveMaximum=1,WaveEase=SineEaseInOut)
    WaveCurve="(EditorCurveData=(Keys=((InterpMode=RCIM_Linear,TangentMode=RCTM_Auto,TangentWeightMode=RCTWM_WeightedNone,Time=0,Value=0.2,ArriveTangent=0,ArriveTangentWeight=0,LeaveTangent=0,LeaveTangentWeight=0),(InterpMode=RCIM_Linear,TangentMode=RCTM_Auto,TangentWeig ...(truncated)"
    bDrawDebug=False
    Speed.Y=RigVMVariableNode() {variable=WaveSpeed 2}.Value
   )
   > &RigUnit_ChainHarmonicsPerItem(
    ChainRoot=(Type=Bone,Name="dyn_hairStrand_upper_G_02")
    DrawWorldOffset=(Rotation=(X=0,Y=0,Z=0,W=1),Translation=(X=0,Y=0,Z=0),Scale3D=(X=1,Y=1,Z=1))
    Pendulum=(bEnabled=false,PendulumStiffness=50,PendulumGravity=(X=-900,Y=-500,Z=0),PendulumBlend=1,PendulumDrag=0.98,PendulumMinimum=0,PendulumMaximum=0.9,PendulumEase=SineEaseInOut,UnwindAxis=(X=0,Y=0,Z=0),UnwindMinimum=0,UnwindMaximum=0)
    Reach=(bEnabled=false,ReachTarget=(X=0,Y=0,Z=0),ReachAxis=(X=1,Y=1,Z=1),ReachMinimum=0,ReachMaximum=0,ReachEase=SineEaseInOut)
    Speed=(X=2,Y=0,Z=2)
    Wave=(bEnabled=True,WaveFrequency=(X=1,Y=0,Z=1),WaveAmplitude=(X=1,Y=0,Z=1),WaveOffset=(X=2,Y=0,Z=2),WaveNoise=(X=0,Y=0,Z=0),WaveMinimum=-1,WaveMaximum=1,WaveEase=SineEaseInOut)
    WaveCurve="(EditorCurveData=(Keys=((InterpMode=RCIM_Linear,TangentMode=RCTM_Auto,TangentWeightMode=RCTWM_WeightedNone,Time=0,Value=0.2,ArriveTangent=0,ArriveTangentWeight=0,LeaveTangent=0,LeaveTangentWeight=0),(InterpMode=RCIM_Linear,TangentMode=RCTM_Auto,TangentWeig ...(truncated)"
    bDrawDebug=False
   )
   > &RigUnit_ChainHarmonicsPerItem(
    ChainRoot=(Type=Bone,Name="dyn_hairStrand_upper_H_01")
    DrawWorldOffset=(Rotation=(X=0,Y=0,Z=0,W=1),Translation=(X=0,Y=0,Z=0),Scale3D=(X=1,Y=1,Z=1))
    Pendulum=(bEnabled=false,PendulumStiffness=50,PendulumGravity=(X=-900,Y=-500,Z=0),PendulumBlend=1,PendulumDrag=20,PendulumMinimum=0,PendulumMaximum=0.8,PendulumEase=SineEaseInOut,UnwindAxis=(X=0,Y=0,Z=0),UnwindMinimum=0,UnwindMaximum=0)
    Reach=(bEnabled=false,ReachTarget=(X=0,Y=0,Z=0),ReachAxis=(X=1,Y=1,Z=1),ReachMinimum=0,ReachMaximum=0,ReachEase=SineEaseInOut)
    Speed=(X=0,Y=2,Z=0)
    Wave=(bEnabled=True,WaveFrequency=(X=0,Y=10,Z=0),WaveAmplitude=(X=0,Y=-3,Z=0),WaveOffset=(X=0,Y=2.3,Z=0),WaveNoise=(X=0,Y=0,Z=0),WaveMinimum=0.1,WaveMaximum=1,WaveEase=SineEaseInOut)
    WaveCurve="(EditorCurveData=(Keys=((InterpMode=RCIM_Linear,TangentMode=RCTM_Auto,TangentWeightMode=RCTWM_WeightedNone,Time=0,Value=0.2,ArriveTangent=0,ArriveTangentWeight=0,LeaveTangent=0,LeaveTangentWeight=0),(InterpMode=RCIM_Linear,TangentMode=RCTM_Auto,TangentWeig ...(truncated)"
    bDrawDebug=False
    Speed.Y=RigVMVariableNode() {variable=WaveSpeed 2}.Value
   )
   > &RigUnit_ChainHarmonicsPerItem(
    ChainRoot=(Type=Bone,Name="dyn_hairStrand_upper_I_01")
    DrawWorldOffset=(Rotation=(X=0,Y=0,Z=0,W=1),Translation=(X=0,Y=0,Z=0),Scale3D=(X=1,Y=1,Z=1))
    Pendulum=(bEnabled=false,PendulumStiffness=50,PendulumGravity=(X=-900,Y=-500,Z=0),PendulumBlend=1,PendulumDrag=20,PendulumMinimum=0,PendulumMaximum=0.8,PendulumEase=SineEaseInOut,UnwindAxis=(X=0,Y=0,Z=0),UnwindMinimum=0,UnwindMaximum=0)
    Reach=(bEnabled=false,ReachTarget=(X=0,Y=0,Z=0),ReachAxis=(X=1,Y=1,Z=1),ReachMinimum=0,ReachMaximum=0,ReachEase=SineEaseInOut)
    Speed=(X=0,Y=2,Z=0)
    Wave=(bEnabled=True,WaveFrequency=(X=0,Y=10,Z=0),WaveAmplitude=(X=0,Y=-2,Z=0),WaveOffset=(X=0,Y=2.5,Z=0),WaveNoise=(X=0,Y=0,Z=0),WaveMinimum=0.1,WaveMaximum=1,WaveEase=SineEaseInOut)
    WaveCurve="(EditorCurveData=(Keys=((InterpMode=RCIM_Linear,TangentMode=RCTM_Auto,TangentWeightMode=RCTWM_WeightedNone,Time=0,Value=0.2,ArriveTangent=0,ArriveTangentWeight=0,LeaveTangent=0,LeaveTangentWeight=0),(InterpMode=RCIM_Linear,TangentMode=RCTM_Auto,TangentWeig ...(truncated)"
    bDrawDebug=False
    Speed.Y=RigVMVariableNode() {variable=WaveSpeed 2}.Value
   )
   > &RigUnit_ChainHarmonicsPerItem(
    ChainRoot=(Type=Bone,Name="dyn_hairStrand_upper_I_02")
    DrawWorldOffset=(Rotation=(X=0,Y=0,Z=0,W=1),Translation=(X=0,Y=0,Z=0),Scale3D=(X=1,Y=1,Z=1))
    Pendulum=(bEnabled=false,PendulumStiffness=50,PendulumGravity=(X=-900,Y=-500,Z=0),PendulumBlend=1,PendulumDrag=0.98,PendulumMinimum=0,PendulumMaximum=0.9,PendulumEase=SineEaseInOut,UnwindAxis=(X=0,Y=0,Z=0),UnwindMinimum=0,UnwindMaximum=0)
    Reach=(bEnabled=false,ReachTarget=(X=0,Y=0,Z=0),ReachAxis=(X=1,Y=1,Z=1),ReachMinimum=0,ReachMaximum=0,ReachEase=SineEaseInOut)
    Speed=(X=2,Y=0,Z=2)
    Wave=(bEnabled=True,WaveFrequency=(X=1,Y=0,Z=1),WaveAmplitude=(X=1,Y=0,Z=1),WaveOffset=(X=2,Y=0,Z=2),WaveNoise=(X=0,Y=0,Z=0),WaveMinimum=-1,WaveMaximum=1,WaveEase=SineEaseInOut)
    WaveCurve="(EditorCurveData=(Keys=((InterpMode=RCIM_Linear,TangentMode=RCTM_Auto,TangentWeightMode=RCTWM_WeightedNone,Time=0,Value=0.2,ArriveTangent=0,ArriveTangentWeight=0,LeaveTangent=0,LeaveTangentWeight=0),(InterpMode=RCIM_Linear,TangentMode=RCTM_Auto,TangentWeig ...(truncated)"
    bDrawDebug=False
   )
   > &RigUnit_ChainHarmonicsPerItem(
    ChainRoot=(Type=Bone,Name="dyn_hairStrand_upper_J_01")
    DrawWorldOffset=(Rotation=(X=0,Y=0,Z=0,W=1),Translation=(X=0,Y=0,Z=0),Scale3D=(X=1,Y=1,Z=1))
    Pendulum=(bEnabled=false,PendulumStiffness=50,PendulumGravity=(X=-900,Y=-500,Z=0),PendulumBlend=1,PendulumDrag=20,PendulumMinimum=0,PendulumMaximum=0.8,PendulumEase=SineEaseInOut,UnwindAxis=(X=0,Y=0,Z=0),UnwindMinimum=0,UnwindMaximum=0)
    Reach=(bEnabled=false,ReachTarget=(X=0,Y=0,Z=0),ReachAxis=(X=1,Y=1,Z=1),ReachMinimum=0,ReachMaximum=0,ReachEase=SineEaseInOut)
    Speed=(X=0,Y=2,Z=0)
    Wave=(bEnabled=True,WaveFrequency=(X=0,Y=10,Z=0),WaveAmplitude=(X=0,Y=-1,Z=0),WaveOffset=(X=0,Y=1.5,Z=0),WaveNoise=(X=0,Y=0,Z=0),WaveMinimum=0.1,WaveMaximum=1,WaveEase=SineEaseInOut)
    WaveCurve="(EditorCurveData=(Keys=((InterpMode=RCIM_Linear,TangentMode=RCTM_Auto,TangentWeightMode=RCTWM_WeightedNone,Time=0,Value=0.2,ArriveTangent=0,ArriveTangentWeight=0,LeaveTangent=0,LeaveTangentWeight=0),(InterpMode=RCIM_Linear,TangentMode=RCTM_Auto,TangentWeig ...(truncated)"
    bDrawDebug=False
    Speed.Y=RigVMVariableNode() {variable=WaveSpeed 2}.Value
   )
   > &RigUnit_ChainHarmonicsPerItem(
    ChainRoot=(Type=Bone,Name="dyn_hairStrand_upper_K_01")
    DrawWorldOffset=(Rotation=(X=0,Y=0,Z=0,W=1),Translation=(X=0,Y=0,Z=0),Scale3D=(X=1,Y=1,Z=1))
    Pendulum=(bEnabled=false,PendulumStiffness=50,PendulumGravity=(X=-900,Y=-500,Z=0),PendulumBlend=1,PendulumDrag=20,PendulumMinimum=0,PendulumMaximum=0.8,PendulumEase=SineEaseInOut,UnwindAxis=(X=0,Y=0,Z=0),UnwindMinimum=0,UnwindMaximum=0)
    Reach=(bEnabled=false,ReachTarget=(X=0,Y=0,Z=0),ReachAxis=(X=1,Y=1,Z=1),ReachMinimum=0,ReachMaximum=0,ReachEase=SineEaseInOut)
    Speed=(X=0,Y=2,Z=0)
    Wave=(bEnabled=True,WaveFrequency=(X=0,Y=10,Z=0),WaveAmplitude=(X=0,Y=-2,Z=0),WaveOffset=(X=0,Y=1,Z=0),WaveNoise=(X=0,Y=0,Z=0),WaveMinimum=0.1,WaveMaximum=1,WaveEase=SineEaseInOut)
    WaveCurve="(EditorCurveData=(Keys=((InterpMode=RCIM_Linear,TangentMode=RCTM_Auto,TangentWeightMode=RCTWM_WeightedNone,Time=0,Value=0.2,ArriveTangent=0,ArriveTangentWeight=0,LeaveTangent=0,LeaveTangentWeight=0),(InterpMode=RCIM_Linear,TangentMode=RCTM_Auto,TangentWeig ...(truncated)"
    bDrawDebug=False
   )
   &RigUnit_ChainHarmonicsPerItem(
    ChainRoot=(Type=Bone,Name="dyn_hairStrand_lower_A_01")
    DrawWorldOffset=(Rotation=(X=0,Y=0,Z=0,W=1),Translation=(X=0,Y=0,Z=0),Scale3D=(X=1,Y=1,Z=1))
    Pendulum=(bEnabled=false,PendulumStiffness=50,PendulumGravity=(X=-900,Y=-500,Z=0),PendulumBlend=1,PendulumDrag=20,PendulumMinimum=0,PendulumMaximum=0.8,PendulumEase=SineEaseInOut,UnwindAxis=(X=0,Y=0,Z=0),UnwindMinimum=0,UnwindMaximum=0)
    Reach=(bEnabled=false,ReachTarget=(X=0,Y=0,Z=0),ReachAxis=(X=1,Y=1,Z=1),ReachMinimum=0,ReachMaximum=0,ReachEase=SineEaseInOut)
    Speed=(X=0,Y=2,Z=0)
    Wave=(bEnabled=True,WaveFrequency=(X=0,Y=10,Z=0),WaveAmplitude=(X=0,Y=-2,Z=0),WaveOffset=(X=0,Y=2,Z=0),WaveNoise=(X=0,Y=0,Z=0),WaveMinimum=0.1,WaveMaximum=1,WaveEase=SineEaseInOut)
    WaveCurve="(EditorCurveData=(Keys=((InterpMode=RCIM_Linear,TangentMode=RCTM_Auto,TangentWeightMode=RCTWM_WeightedNone,Time=0,Value=0.2,ArriveTangent=0,ArriveTangentWeight=0,LeaveTangent=0,LeaveTangentWeight=0),(InterpMode=RCIM_Linear,TangentMode=RCTM_Auto,TangentWeig ...(truncated)"
    bDrawDebug=False
    Speed.Y=RigVMVariableNode() {variable=WaveSpeed 2}.Value
    ExecutePin=@RigVMFunction_Sequence.A
   )
   > &RigUnit_ChainHarmonicsPerItem(
    ChainRoot=(Type=Bone,Name="dyn_hairStrand_lower_A_02")
    DrawWorldOffset=(Rotation=(X=0,Y=0,Z=0,W=1),Translation=(X=0,Y=0,Z=0),Scale3D=(X=1,Y=1,Z=1))
    Pendulum=(bEnabled=false,PendulumStiffness=50,PendulumGravity=(X=-900,Y=-500,Z=0),PendulumBlend=1,PendulumDrag=0.98,PendulumMinimum=0,PendulumMaximum=0.9,PendulumEase=SineEaseInOut,UnwindAxis=(X=0,Y=0,Z=0),UnwindMinimum=0,UnwindMaximum=0)
    Reach=(bEnabled=false,ReachTarget=(X=0,Y=0,Z=0),ReachAxis=(X=1,Y=1,Z=1),ReachMinimum=0,ReachMaximum=0,ReachEase=SineEaseInOut)
    Speed=(X=2,Y=0,Z=2)
    Wave=(bEnabled=True,WaveFrequency=(X=1,Y=0,Z=1),WaveAmplitude=(X=1,Y=0,Z=1),WaveOffset=(X=2,Y=0,Z=2),WaveNoise=(X=0,Y=0,Z=0),WaveMinimum=-1,WaveMaximum=1,WaveEase=SineEaseInOut)
    WaveCurve="(EditorCurveData=(Keys=((InterpMode=RCIM_Linear,TangentMode=RCTM_Auto,TangentWeightMode=RCTWM_WeightedNone,Time=0,Value=0.2,ArriveTangent=0,ArriveTangentWeight=0,LeaveTangent=0,LeaveTangentWeight=0),(InterpMode=RCIM_Linear,TangentMode=RCTM_Auto,TangentWeig ...(truncated)"
    bDrawDebug=False
   )
   > &RigUnit_ChainHarmonicsPerItem(
    ChainRoot=(Type=Bone,Name="dyn_hairStrand_lower_B_01")
    DrawWorldOffset=(Rotation=(X=0,Y=0,Z=0,W=1),Translation=(X=0,Y=0,Z=0),Scale3D=(X=1,Y=1,Z=1))
    Pendulum=(bEnabled=false,PendulumStiffness=50,PendulumGravity=(X=-900,Y=-500,Z=0),PendulumBlend=1,PendulumDrag=20,PendulumMinimum=0,PendulumMaximum=0.8,PendulumEase=SineEaseInOut,UnwindAxis=(X=0,Y=0,Z=0),UnwindMinimum=0,UnwindMaximum=0)
    Reach=(bEnabled=false,ReachTarget=(X=0,Y=0,Z=0),ReachAxis=(X=1,Y=1,Z=1),ReachMinimum=0,ReachMaximum=0,ReachEase=SineEaseInOut)
    Speed=(X=0,Y=2,Z=0)
    Wave=(bEnabled=True,WaveFrequency=(X=0,Y=10,Z=0),WaveAmplitude=(X=0,Y=-2,Z=0),WaveOffset=(X=0,Y=1.5,Z=0),WaveNoise=(X=0,Y=0,Z=0),WaveMinimum=0.1,WaveMaximum=1,WaveEase=SineEaseInOut)
    WaveCurve="(EditorCurveData=(Keys=((InterpMode=RCIM_Linear,TangentMode=RCTM_Auto,TangentWeightMode=RCTWM_WeightedNone,Time=0,Value=0.2,ArriveTangent=0,ArriveTangentWeight=0,LeaveTangent=0,LeaveTangentWeight=0),(InterpMode=RCIM_Linear,TangentMode=RCTM_Auto,TangentWeig ...(truncated)"
    bDrawDebug=False
    Speed.Y=RigVMVariableNode() {variable=WaveSpeed 2}.Value
   )
   > &RigUnit_ChainHarmonicsPerItem(
    ChainRoot=(Type=Bone,Name="dyn_hairStrand_lower_B_02")
    DrawWorldOffset=(Rotation=(X=0,Y=0,Z=0,W=1),Translation=(X=0,Y=0,Z=0),Scale3D=(X=1,Y=1,Z=1))
    Pendulum=(bEnabled=false,PendulumStiffness=50,PendulumGravity=(X=-900,Y=-500,Z=0),PendulumBlend=1,PendulumDrag=0.98,PendulumMinimum=0,PendulumMaximum=0.9,PendulumEase=SineEaseInOut,UnwindAxis=(X=0,Y=0,Z=0),UnwindMinimum=0,UnwindMaximum=0)
    Reach=(bEnabled=false,ReachTarget=(X=0,Y=0,Z=0),ReachAxis=(X=1,Y=1,Z=1),ReachMinimum=0,ReachMaximum=0,ReachEase=SineEaseInOut)
    Speed=(X=2,Y=0,Z=2)
    Wave=(bEnabled=True,WaveFrequency=(X=1,Y=0,Z=1),WaveAmplitude=(X=1,Y=0,Z=1),WaveOffset=(X=0,Y=0,Z=0),WaveNoise=(X=0,Y=0,Z=0),WaveMinimum=-1,WaveMaximum=1,WaveEase=SineEaseInOut)
    WaveCurve="(EditorCurveData=(Keys=((InterpMode=RCIM_Linear,TangentMode=RCTM_Auto,TangentWeightMode=RCTWM_WeightedNone,Time=0,Value=0.2,ArriveTangent=0,ArriveTangentWeight=0,LeaveTangent=0,LeaveTangentWeight=0),(InterpMode=RCIM_Linear,TangentMode=RCTM_Auto,TangentWeig ...(truncated)"
    bDrawDebug=False
   )
   > &RigUnit_ChainHarmonicsPerItem(
    ChainRoot=(Type=Bone,Name="dyn_hairStrand_lower_C_01")
    DrawWorldOffset=(Rotation=(X=0,Y=0,Z=0,W=1),Translation=(X=0,Y=0,Z=0),Scale3D=(X=1,Y=1,Z=1))
    Pendulum=(bEnabled=false,PendulumStiffness=50,PendulumGravity=(X=-900,Y=-500,Z=0),PendulumBlend=1,PendulumDrag=20,PendulumMinimum=0,PendulumMaximum=0.8,PendulumEase=SineEaseInOut,UnwindAxis=(X=0,Y=0,Z=0),UnwindMinimum=0,UnwindMaximum=0)
    Reach=(bEnabled=false,ReachTarget=(X=0,Y=0,Z=0),ReachAxis=(X=1,Y=1,Z=1),ReachMinimum=0,ReachMaximum=0,ReachEase=SineEaseInOut)
    Speed=(X=0,Y=2,Z=0)
    Wave=(bEnabled=True,WaveFrequency=(X=0,Y=10,Z=0),WaveAmplitude=(X=0,Y=-2,Z=0),WaveOffset=(X=0,Y=2,Z=0),WaveNoise=(X=0,Y=0,Z=0),WaveMinimum=0.1,WaveMaximum=1,WaveEase=SineEaseInOut)
    WaveCurve="(EditorCurveData=(Keys=((InterpMode=RCIM_Linear,TangentMode=RCTM_Auto,TangentWeightMode=RCTWM_WeightedNone,Time=0,Value=0.2,ArriveTangent=0,ArriveTangentWeight=0,LeaveTangent=0,LeaveTangentWeight=0),(InterpMode=RCIM_Linear,TangentMode=RCTM_Auto,TangentWeig ...(truncated)"
    bDrawDebug=False
    Speed.Y=RigVMVariableNode() {variable=WaveSpeed 2}.Value
   )
   > &RigUnit_ChainHarmonicsPerItem(
    ChainRoot=(Type=Bone,Name="dyn_hairStrand_lower_C_02")
    DrawWorldOffset=(Rotation=(X=0,Y=0,Z=0,W=1),Translation=(X=0,Y=0,Z=0),Scale3D=(X=1,Y=1,Z=1))
    Pendulum=(bEnabled=false,PendulumStiffness=50,PendulumGravity=(X=-900,Y=-500,Z=0),PendulumBlend=1,PendulumDrag=0.98,PendulumMinimum=0,PendulumMaximum=0.9,PendulumEase=SineEaseInOut,UnwindAxis=(X=0,Y=0,Z=0),UnwindMinimum=0,UnwindMaximum=0)
    Reach=(bEnabled=false,ReachTarget=(X=0,Y=0,Z=0),ReachAxis=(X=1,Y=1,Z=1),ReachMinimum=0,ReachMaximum=0,ReachEase=SineEaseInOut)
    Speed=(X=2,Y=0,Z=2)
    Wave=(bEnabled=True,WaveFrequency=(X=1,Y=0,Z=1),WaveAmplitude=(X=1,Y=0,Z=1),WaveOffset=(X=2,Y=0,Z=2),WaveNoise=(X=0,Y=0,Z=0),WaveMinimum=-1,WaveMaximum=1,WaveEase=SineEaseInOut)
    WaveCurve="(EditorCurveData=(Keys=((InterpMode=RCIM_Linear,TangentMode=RCTM_Auto,TangentWeightMode=RCTWM_WeightedNone,Time=0,Value=0.2,ArriveTangent=0,ArriveTangentWeight=0,LeaveTangent=0,LeaveTangentWeight=0),(InterpMode=RCIM_Linear,TangentMode=RCTM_Auto,TangentWeig ...(truncated)"
    bDrawDebug=False
   )
   > &RigUnit_ChainHarmonicsPerItem(
    ChainRoot=(Type=Bone,Name="dyn_hairStrand_lower_D_01")
    DrawWorldOffset=(Rotation=(X=0,Y=0,Z=0,W=1),Translation=(X=0,Y=0,Z=0),Scale3D=(X=1,Y=1,Z=1))
    Pendulum=(bEnabled=false,PendulumStiffness=50,PendulumGravity=(X=-900,Y=-500,Z=0),PendulumBlend=1,PendulumDrag=20,PendulumMinimum=0,PendulumMaximum=0.8,PendulumEase=SineEaseInOut,UnwindAxis=(X=0,Y=0,Z=0),UnwindMinimum=0,UnwindMaximum=0)
    Reach=(bEnabled=false,ReachTarget=(X=0,Y=0,Z=0),ReachAxis=(X=1,Y=1,Z=1),ReachMinimum=0,ReachMaximum=0,ReachEase=SineEaseInOut)
    Speed=(X=0,Y=2,Z=0)
    Wave=(bEnabled=True,WaveFrequency=(X=0,Y=10,Z=0),WaveAmplitude=(X=0,Y=-3,Z=0),WaveOffset=(X=0,Y=2.3,Z=0),WaveNoise=(X=0,Y=0,Z=0),WaveMinimum=0.1,WaveMaximum=1,WaveEase=SineEaseInOut)
    WaveCurve="(EditorCurveData=(Keys=((InterpMode=RCIM_Linear,TangentMode=RCTM_Auto,TangentWeightMode=RCTWM_WeightedNone,Time=0,Value=0.2,ArriveTangent=0,ArriveTangentWeight=0,LeaveTangent=0,LeaveTangentWeight=0),(InterpMode=RCIM_Linear,TangentMode=RCTM_Auto,TangentWeig ...(truncated)"
    bDrawDebug=False
    Speed.Y=RigVMVariableNode() {variable=WaveSpeed 2}.Value
   )
   > &RigUnit_ChainHarmonicsPerItem(
    ChainRoot=(Type=Bone,Name="dyn_hairStrand_lower_D_02")
    DrawWorldOffset=(Rotation=(X=0,Y=0,Z=0,W=1),Translation=(X=0,Y=0,Z=0),Scale3D=(X=1,Y=1,Z=1))
    Pendulum=(bEnabled=false,PendulumStiffness=50,PendulumGravity=(X=-900,Y=-500,Z=0),PendulumBlend=1,PendulumDrag=0.98,PendulumMinimum=0,PendulumMaximum=0.9,PendulumEase=SineEaseInOut,UnwindAxis=(X=0,Y=0,Z=0),UnwindMinimum=0,UnwindMaximum=0)
    Reach=(bEnabled=false,ReachTarget=(X=0,Y=0,Z=0),ReachAxis=(X=1,Y=1,Z=1),ReachMinimum=0,ReachMaximum=0,ReachEase=SineEaseInOut)
    Speed=(X=2,Y=0,Z=2)
    Wave=(bEnabled=True,WaveFrequency=(X=1,Y=0,Z=1),WaveAmplitude=(X=1,Y=0,Z=1),WaveOffset=(X=0,Y=0,Z=0),WaveNoise=(X=0,Y=0,Z=0),WaveMinimum=-1,WaveMaximum=1,WaveEase=SineEaseInOut)
    WaveCurve="(EditorCurveData=(Keys=((InterpMode=RCIM_Linear,TangentMode=RCTM_Auto,TangentWeightMode=RCTWM_WeightedNone,Time=0,Value=0.2,ArriveTangent=0,ArriveTangentWeight=0,LeaveTangent=0,LeaveTangentWeight=0),(InterpMode=RCIM_Linear,TangentMode=RCTM_Auto,TangentWeig ...(truncated)"
    bDrawDebug=False
   )
   > &RigUnit_ChainHarmonicsPerItem(
    ChainRoot=(Type=Bone,Name="dyn_hairStrand_lower_E_01")
    DrawWorldOffset=(Rotation=(X=0,Y=0,Z=0,W=1),Translation=(X=0,Y=0,Z=0),Scale3D=(X=1,Y=1,Z=1))
    Pendulum=(bEnabled=false,PendulumStiffness=50,PendulumGravity=(X=-900,Y=-500,Z=0),PendulumBlend=1,PendulumDrag=20,PendulumMinimum=0,PendulumMaximum=0.8,PendulumEase=SineEaseInOut,UnwindAxis=(X=0,Y=0,Z=0),UnwindMinimum=0,UnwindMaximum=0)
    Reach=(bEnabled=false,ReachTarget=(X=0,Y=0,Z=0),ReachAxis=(X=1,Y=1,Z=1),ReachMinimum=0,ReachMaximum=0,ReachEase=SineEaseInOut)
    Speed=(X=0,Y=4,Z=0)
    Wave=(bEnabled=True,WaveFrequency=(X=0,Y=10,Z=0),WaveAmplitude=(X=0,Y=-1,Z=0),WaveOffset=(X=0,Y=2,Z=0),WaveNoise=(X=0,Y=0,Z=0),WaveMinimum=0.1,WaveMaximum=1,WaveEase=SineEaseInOut)
    WaveCurve="(EditorCurveData=(Keys=((InterpMode=RCIM_Linear,TangentMode=RCTM_Auto,TangentWeightMode=RCTWM_WeightedNone,Time=0,Value=0.2,ArriveTangent=0,ArriveTangentWeight=0,LeaveTangent=0,LeaveTangentWeight=0),(InterpMode=RCIM_Linear,TangentMode=RCTM_Auto,TangentWeig ...(truncated)"
    bDrawDebug=False
    Speed.Y=RigVMVariableNode() {variable=WaveSpeed 4}.Value
   )
   > &RigUnit_ChainHarmonicsPerItem(
    ChainRoot=(Type=Bone,Name="dyn_hairStrand_lower_E_02")
    DrawWorldOffset=(Rotation=(X=0,Y=0,Z=0,W=1),Translation=(X=0,Y=0,Z=0),Scale3D=(X=1,Y=1,Z=1))
    Pendulum=(bEnabled=false,PendulumStiffness=50,PendulumGravity=(X=-900,Y=-500,Z=0),PendulumBlend=1,PendulumDrag=0.98,PendulumMinimum=0,PendulumMaximum=0.9,PendulumEase=SineEaseInOut,UnwindAxis=(X=0,Y=0,Z=0),UnwindMinimum=0,UnwindMaximum=0)
    Reach=(bEnabled=false,ReachTarget=(X=0,Y=0,Z=0),ReachAxis=(X=1,Y=1,Z=1),ReachMinimum=0,ReachMaximum=0,ReachEase=SineEaseInOut)
    Speed=(X=2,Y=0,Z=2)
    Wave=(bEnabled=True,WaveFrequency=(X=1,Y=0,Z=1),WaveAmplitude=(X=1,Y=0,Z=1),WaveOffset=(X=2,Y=0,Z=2),WaveNoise=(X=0,Y=0,Z=0),WaveMinimum=-1,WaveMaximum=1,WaveEase=SineEaseInOut)
    WaveCurve="(EditorCurveData=(Keys=((InterpMode=RCIM_Linear,TangentMode=RCTM_Auto,TangentWeightMode=RCTWM_WeightedNone,Time=0,Value=0.2,ArriveTangent=0,ArriveTangentWeight=0,LeaveTangent=0,LeaveTangentWeight=0),(InterpMode=RCIM_Linear,TangentMode=RCTM_Auto,TangentWeig ...(truncated)"
    bDrawDebug=True
   )
   > &RigUnit_ChainHarmonicsPerItem(
    ChainRoot=(Type=Bone,Name="dyn_hairStrand_lower_F_01")
    DrawWorldOffset=(Rotation=(X=0,Y=0,Z=0,W=1),Translation=(X=0,Y=0,Z=0),Scale3D=(X=1,Y=1,Z=1))
    Pendulum=(bEnabled=false,PendulumStiffness=50,PendulumGravity=(X=-900,Y=-500,Z=0),PendulumBlend=1,PendulumDrag=20,PendulumMinimum=0,PendulumMaximum=0.8,PendulumEase=SineEaseInOut,UnwindAxis=(X=0,Y=0,Z=0),UnwindMinimum=0,UnwindMaximum=0)
    Reach=(bEnabled=false,ReachTarget=(X=0,Y=0,Z=0),ReachAxis=(X=1,Y=1,Z=1),ReachMinimum=0,ReachMaximum=0,ReachEase=SineEaseInOut)
    Speed=(X=0,Y=2,Z=0)
    Wave=(bEnabled=True,WaveFrequency=(X=0,Y=10,Z=0),WaveAmplitude=(X=0,Y=-3,Z=0),WaveOffset=(X=0,Y=1.5,Z=0),WaveNoise=(X=0,Y=0,Z=0),WaveMinimum=0.1,WaveMaximum=1,WaveEase=SineEaseInOut)
    WaveCurve="(EditorCurveData=(Keys=((InterpMode=RCIM_Linear,TangentMode=RCTM_Auto,TangentWeightMode=RCTWM_WeightedNone,Time=0,Value=0.2,ArriveTangent=0,ArriveTangentWeight=0,LeaveTangent=0,LeaveTangentWeight=0),(InterpMode=RCIM_Linear,TangentMode=RCTM_Auto,TangentWeig ...(truncated)"
    bDrawDebug=False
    Speed.Y=RigVMVariableNode() {variable=WaveSpeed 2}.Value
   )
   > &RigUnit_ChainHarmonicsPerItem(
    ChainRoot=(Type=Bone,Name="dyn_hairStrand_lower_F_02")
    DrawWorldOffset=(Rotation=(X=0,Y=0,Z=0,W=1),Translation=(X=0,Y=0,Z=0),Scale3D=(X=1,Y=1,Z=1))
    Pendulum=(bEnabled=false,PendulumStiffness=50,PendulumGravity=(X=-900,Y=-500,Z=0),PendulumBlend=1,PendulumDrag=0.98,PendulumMinimum=0,PendulumMaximum=0.9,PendulumEase=SineEaseInOut,UnwindAxis=(X=0,Y=0,Z=0),UnwindMinimum=0,UnwindMaximum=0)
    Reach=(bEnabled=false,ReachTarget=(X=0,Y=0,Z=0),ReachAxis=(X=1,Y=1,Z=1),ReachMinimum=0,ReachMaximum=0,ReachEase=SineEaseInOut)
    Speed=(X=2,Y=0,Z=2)
    Wave=(bEnabled=True,WaveFrequency=(X=1,Y=0,Z=1),WaveAmplitude=(X=1,Y=0,Z=1),WaveOffset=(X=0,Y=0,Z=0),WaveNoise=(X=0,Y=0,Z=0),WaveMinimum=-1,WaveMaximum=1,WaveEase=SineEaseInOut)
    WaveCurve="(EditorCurveData=(Keys=((InterpMode=RCIM_Linear,TangentMode=RCTM_Auto,TangentWeightMode=RCTWM_WeightedNone,Time=0,Value=0.2,ArriveTangent=0,ArriveTangentWeight=0,LeaveTangent=0,LeaveTangentWeight=0),(InterpMode=RCIM_Linear,TangentMode=RCTM_Auto,TangentWeig ...(truncated)"
    bDrawDebug=False
   )
   > &RigUnit_ChainHarmonicsPerItem(
    ChainRoot=(Type=Bone,Name="dyn_hairStrand_lower_G_01")
    DrawWorldOffset=(Rotation=(X=0,Y=0,Z=0,W=1),Translation=(X=0,Y=0,Z=0),Scale3D=(X=1,Y=1,Z=1))
    Pendulum=(bEnabled=false,PendulumStiffness=50,PendulumGravity=(X=-900,Y=-500,Z=0),PendulumBlend=1,PendulumDrag=20,PendulumMinimum=0,PendulumMaximum=0.8,PendulumEase=SineEaseInOut,UnwindAxis=(X=0,Y=0,Z=0),UnwindMinimum=0,UnwindMaximum=0)
    Reach=(bEnabled=false,ReachTarget=(X=0,Y=0,Z=0),ReachAxis=(X=1,Y=1,Z=1),ReachMinimum=0,ReachMaximum=0,ReachEase=SineEaseInOut)
    Speed=(X=0,Y=3,Z=0)
    Wave=(bEnabled=True,WaveFrequency=(X=0,Y=10,Z=0),WaveAmplitude=(X=0,Y=-2,Z=0),WaveOffset=(X=0,Y=2,Z=0),WaveNoise=(X=0,Y=0,Z=0),WaveMinimum=0.1,WaveMaximum=1,WaveEase=SineEaseInOut)
    WaveCurve="(EditorCurveData=(Keys=((InterpMode=RCIM_Linear,TangentMode=RCTM_Auto,TangentWeightMode=RCTWM_WeightedNone,Time=0,Value=0.2,ArriveTangent=0,ArriveTangentWeight=0,LeaveTangent=0,LeaveTangentWeight=0),(InterpMode=RCIM_Linear,TangentMode=RCTM_Auto,TangentWeig ...(truncated)"
    bDrawDebug=False
    Speed.Y=RigVMVariableNode() {variable=WaveSpeed 3}.Value
   )
   > &RigUnit_ChainHarmonicsPerItem(
    ChainRoot=(Type=Bone,Name="dyn_hairStrand_lower_H")
    DrawWorldOffset=(Rotation=(X=0,Y=0,Z=0,W=1),Translation=(X=0,Y=0,Z=0),Scale3D=(X=1,Y=1,Z=1))
    Pendulum=(bEnabled=false,PendulumStiffness=50,PendulumGravity=(X=-900,Y=-500,Z=0),PendulumBlend=1,PendulumDrag=20,PendulumMinimum=0,PendulumMaximum=0.8,PendulumEase=SineEaseInOut,UnwindAxis=(X=0,Y=0,Z=0),UnwindMinimum=0,UnwindMaximum=0)
    Reach=(bEnabled=false,ReachTarget=(X=0,Y=0,Z=0),ReachAxis=(X=1,Y=1,Z=1),ReachMinimum=0,ReachMaximum=0,ReachEase=SineEaseInOut)
    Speed=(X=0,Y=2,Z=0)
    Wave=(bEnabled=True,WaveFrequency=(X=0,Y=10,Z=0),WaveAmplitude=(X=0,Y=-1,Z=0),WaveOffset=(X=0,Y=2.3,Z=0),WaveNoise=(X=0,Y=0,Z=0),WaveMinimum=0.1,WaveMaximum=1,WaveEase=SineEaseInOut)
    WaveCurve="(EditorCurveData=(Keys=((InterpMode=RCIM_Linear,TangentMode=RCTM_Auto,TangentWeightMode=RCTWM_WeightedNone,Time=0,Value=0.2,ArriveTangent=0,ArriveTangentWeight=0,LeaveTangent=0,LeaveTangentWeight=0),(InterpMode=RCIM_Linear,TangentMode=RCTM_Auto,TangentWeig ...(truncated)"
    bDrawDebug=False
    Speed.Y=RigVMVariableNode() {variable=WaveSpeed 2}.Value
   )
   &RigUnit_ChainHarmonicsPerItem(
    ChainRoot=(Type=Bone,Name="dyn_bang_A_01")
    DrawWorldOffset=(Rotation=(X=0,Y=0,Z=0,W=1),Translation=(X=0,Y=0,Z=0),Scale3D=(X=1,Y=1,Z=1))
    Pendulum=(bEnabled=false,PendulumStiffness=50,PendulumGravity=(X=-900,Y=-500,Z=0),PendulumBlend=1,PendulumDrag=20,PendulumMinimum=0,PendulumMaximum=0.8,PendulumEase=SineEaseInOut,UnwindAxis=(X=0,Y=0,Z=0),UnwindMinimum=0,UnwindMaximum=0)
    Reach=(bEnabled=false,ReachTarget=(X=0,Y=0,Z=0),ReachAxis=(X=1,Y=1,Z=1),ReachMinimum=0,ReachMaximum=0,ReachEase=SineEaseInOut)
    Speed=(X=0,Y=2,Z=0)
    Wave=(bEnabled=True,WaveFrequency=(X=0,Y=10,Z=0),WaveAmplitude=(X=0,Y=-1,Z=0),WaveOffset=(X=0,Y=2,Z=0),WaveNoise=(X=0,Y=0,Z=0),WaveMinimum=0.1,WaveMaximum=1,WaveEase=SineEaseInOut)
    WaveCurve="(EditorCurveData=(Keys=((InterpMode=RCIM_Linear,TangentMode=RCTM_Auto,TangentWeightMode=RCTWM_WeightedNone,Time=0,Value=0.2,ArriveTangent=0,ArriveTangentWeight=0,LeaveTangent=0,LeaveTangentWeight=0),(InterpMode=RCIM_Linear,TangentMode=RCTM_Auto,TangentWeig ...(truncated)"
    bDrawDebug=False
    Speed.Y=RigVMVariableNode() {variable=WaveSpeed 2}.Value
    ExecutePin=@RigVMFunction_Sequence.D
   )
   > &RigUnit_ChainHarmonicsPerItem(
    ChainRoot=(Type=Bone,Name="dyn_bang_B_01")
    DrawWorldOffset=(Rotation=(X=0,Y=0,Z=0,W=1),Translation=(X=0,Y=0,Z=0),Scale3D=(X=1,Y=1,Z=1))
    Pendulum=(bEnabled=false,PendulumStiffness=50,PendulumGravity=(X=-900,Y=-500,Z=0),PendulumBlend=1,PendulumDrag=20,PendulumMinimum=0,PendulumMaximum=0.8,PendulumEase=SineEaseInOut,UnwindAxis=(X=0,Y=0,Z=0),UnwindMinimum=0,UnwindMaximum=0)
    Reach=(bEnabled=false,ReachTarget=(X=0,Y=0,Z=0),ReachAxis=(X=1,Y=1,Z=1),ReachMinimum=0,ReachMaximum=0,ReachEase=SineEaseInOut)
    Speed=(X=0,Y=2,Z=0)
    Wave=(bEnabled=True,WaveFrequency=(X=0,Y=10,Z=0),WaveAmplitude=(X=0,Y=-1,Z=0),WaveOffset=(X=0,Y=1.5,Z=0),WaveNoise=(X=0,Y=0,Z=0),WaveMinimum=0.1,WaveMaximum=1,WaveEase=SineEaseInOut)
    WaveCurve="(EditorCurveData=(Keys=((InterpMode=RCIM_Linear,TangentMode=RCTM_Auto,TangentWeightMode=RCTWM_WeightedNone,Time=0,Value=0.2,ArriveTangent=0,ArriveTangentWeight=0,LeaveTangent=0,LeaveTangentWeight=0),(InterpMode=RCIM_Linear,TangentMode=RCTM_Auto,TangentWeig ...(truncated)"
    bDrawDebug=False
    Speed.Y=RigVMVariableNode() {variable=WaveSpeed 2}.Value
   )
   > &RigUnit_ChainHarmonicsPerItem(
    ChainRoot=(Type=Bone,Name="dyn_bang_C_01")
    DrawWorldOffset=(Rotation=(X=0,Y=0,Z=0,W=1),Translation=(X=0,Y=0,Z=0),Scale3D=(X=1,Y=1,Z=1))
    Pendulum=(bEnabled=false,PendulumStiffness=50,PendulumGravity=(X=-900,Y=-500,Z=0),PendulumBlend=1,PendulumDrag=20,PendulumMinimum=0,PendulumMaximum=0.8,PendulumEase=SineEaseInOut,UnwindAxis=(X=0,Y=0,Z=0),UnwindMinimum=0,UnwindMaximum=0)
    Reach=(bEnabled=false,ReachTarget=(X=0,Y=0,Z=0),ReachAxis=(X=1,Y=1,Z=1),ReachMinimum=0,ReachMaximum=0,ReachEase=SineEaseInOut)
    Speed=(X=0,Y=3,Z=0)
    Wave=(bEnabled=True,WaveFrequency=(X=0,Y=10,Z=0),WaveAmplitude=(X=0,Y=-0.3,Z=0),WaveOffset=(X=0,Y=-2,Z=0),WaveNoise=(X=0,Y=0,Z=0),WaveMinimum=0.1,WaveMaximum=1,WaveEase=SineEaseInOut)
    WaveCurve="(EditorCurveData=(Keys=((InterpMode=RCIM_Linear,TangentMode=RCTM_Auto,TangentWeightMode=RCTWM_WeightedNone,Time=0,Value=0.2,ArriveTangent=0,ArriveTangentWeight=0,LeaveTangent=0,LeaveTangentWeight=0),(InterpMode=RCIM_Linear,TangentMode=RCTM_Auto,TangentWeig ...(truncated)"
    bDrawDebug=False
    Speed.Y=RigVMVariableNode() {variable=WaveSpeed 3}.Value
   )
  ]}
]
hierarchy [
 root {loc=<0 0 0>, rot=<0 0 0 1>}
 pelvis {parent=root, loc=<0 -0.5301 98.0571>, rot=<0.0127 0.707 -0.0127 -0.707>}
 spine_01 {parent=pelvis, loc=<6.1208 -0.9328 0>, rot=<0 0 0.034 -0.9994>}
 spine_02 {parent=spine_01, loc=<6.5997 -0.2095 0>, rot=<0 0 0.0238 -0.9997>}
 spine_03 {parent=spine_02, loc=<6.4914 -0.1237 0>, rot=<0 0 0.0149 -0.9999>}
 spine_04 {parent=spine_03, loc=<6.9598 0.0794 0>, rot=<0 0 -0.0132 -0.9999>}
 spine_05 {parent=spine_04, loc=<6.9987 0 0>, rot=<0 0 -0.0015 -1>}
 neck_01 {parent=spine_05, loc=<11.3138 0.1529 0>, rot=<0 0 -0.2003 -0.9797>}
 neck_02 {parent=neck_01, loc=<4.5627 0 0>, rot=<0 0 0.0504 -0.9987>}
 head {parent=neck_02, loc=<5.3914 0 0>, rot=<0 0 0.115 -0.9934>}
 dyn_hair_Mstr_r {parent=head, loc=<-2.5327 -4.7944 8.0012>, rot=<0.5757 -0.0391 0.8167 0.0038>}
 dyn_hairStrand_upper_D_01 {parent=dyn_hair_Mstr_r, loc=<1.7738 0.3128 3.8626>, rot=<0.0292 0.1289 -0.2635 -0.9556>}
 dyn_hairStrand_upper_D_02 {parent=dyn_hairStrand_upper_D_01, loc=<5.4459 0 0>, rot=<0.1227 0.3391 -0.0446 -0.9317>}
 dyn_hairStrand_lower_B_01 {parent=dyn_hair_Mstr_r, loc=<1.0671 -4.2885 2.2728>, rot=<-0.0772 -0.2853 0.0552 -0.9537>}
 dyn_hairStrand_lower_B_02 {parent=dyn_hairStrand_lower_B_01, loc=<5.4129 0 0>, rot=<0.1019 0.5103 -0.2721 -0.8094>}
 dyn_hairStrand_lower_B_03 {parent=dyn_hairStrand_lower_B_02, loc=<3.9087 0 0>, rot=<0 0.3827 0 -0.9239>}
 dyn_hairStrand_lower_C_01 {parent=dyn_hair_Mstr_r, loc=<2.302 -3.7886 -2.31>, rot=<0.1524 -0.2499 -0.1236 -0.9482>}
 dyn_hairStrand_lower_C_02 {parent=dyn_hairStrand_lower_C_01, loc=<3.1471 -0.2182 0.8145>, rot=<0.0819 0.3214 0.117 -0.9361>}
 dyn_hairStrand_lower_D_01 {parent=dyn_hair_Mstr_r, loc=<2.8078 -1.8185 0.5721>, rot=<-0.0531 -0.0216 -0.2911 -0.955>}
 dyn_hairStrand_lower_D_02 {parent=dyn_hairStrand_lower_D_01, loc=<4.0829 0 0>, rot=<0.0238 0.2476 -0.0274 -0.9682>}
 dyn_hairStrand_lower_D_03 {parent=dyn_hairStrand_lower_D_02, loc=<4.2655 0.3176 0.8074>, rot=<-0.0043 0.5105 -0.0137 -0.8597>}
 dyn_hairStrand_lower_E_01 {parent=dyn_hair_Mstr_r, loc=<1.0682 0.1301 -1.8166>, rot=<0.0595 -0.0571 -0.3074 -0.948>}
 dyn_hairStrand_lower_E_02 {parent=dyn_hairStrand_lower_E_01, loc=<6.6113 0 0>, rot=<0 0.3827 0 -0.9239>}
 dyn_hairStrand_lower_F_01 {parent=dyn_hair_Mstr_r, loc=<1.7373 4.6445 -0.4322>, rot=<-0.1436 0.1123 -0.2956 -0.9378>}
 dyn_hairStrand_lower_F_02 {parent=dyn_hairStrand_lower_F_01, loc=<5.5559 0.3351 0.9206>, rot=<0.0002 0.4421 -0.0611 -0.8949>}
 dyn_hair_Mstr_mid {parent=head, loc=<-4.0229 -9.6729 0>, rot=<0.3359 -0.6152 0.6015 -0.3831>}
 dyn_hairStrand_lower_G_01 {parent=dyn_hair_Mstr_mid, loc=<0.9428 -2.9451 -0.1175>, rot=<-0.1151 0.028 0.2657 -0.9567>}
 dyn_hairStrand_lower_G_02 {parent=dyn_hairStrand_lower_G_01, loc=<3.7559 0 0>, rot=<0.0198 0.2981 0.0058 -0.9543>}
 dyn_hairStrand_lower_G_03 {parent=dyn_hairStrand_lower_G_02, loc=<4.0705 -0.379 1.4815>, rot=<0.1224 0.3734 0.0931 -0.9149>}
 dyn_hairStrand_lower_G_04 {parent=dyn_hairStrand_lower_G_02, loc=<2.7397 3.3943 -0.8099>, rot=<0.0076 0.2578 -0.0868 -0.9623>}
 dyn_hairStrand_lower_G_05 {parent=dyn_hairStrand_lower_G_02, loc=<3.7048 -2.18 -0.6061>, rot=<0 0.3391 0.1305 -0.9317>}
 dyn_hairStrand_lower_H {parent=dyn_hair_Mstr_mid, loc=<4.4372 0.3858 -1.6687>, rot=<0.0397 0.1593 -0.1216 -0.9789>}
 dyn_hairStrand_lower_I_01 {parent=dyn_hair_Mstr_mid, loc=<4.2179 4.6639 -0.2146>, rot=<0.0748 0.2284 -0.1573 -0.9579>}
 dyn_hairStrand_lower_I_02 {parent=dyn_hairStrand_lower_I_01, loc=<6.1382 -0.5453 1.0823>, rot=<-0.0605 0.5012 -0.0027 -0.8632>}
 dyn_hairStrand_lower_I_03 {parent=dyn_hairStrand_lower_I_01, loc=<6.4102 2.2349 0.6128>, rot=<-0.0149 0.4995 -0.041 -0.8652>}
 dyn_hair_Mstr_l {parent=head, loc=<-2.5305 -5.0642 -6.4682>, rot=<0.245 -0.744 0.3831 -0.4896>}
 dyn_hairStrand_upper_H_01 {parent=dyn_hair_Mstr_l, loc=<0.0487 2.6122 4.0577>, rot=<0.0607 0.0727 -0.1471 -0.9846>}
 dyn_hairStrand_upper_H_02 {parent=dyn_hairStrand_upper_H_01, loc=<4.0078 0 0>, rot=<-0.058 0.2596 0.0656 -0.9617>}
 dyn_hairStrand_upper_H_03 {parent=dyn_hairStrand_upper_H_02, loc=<3.0619 1.4278 0>, rot=<0.1069 0.5036 -0.0626 -0.855>}
 dyn_hairStrand_upper_H_04 {parent=dyn_hairStrand_upper_H_02, loc=<3.9462 -1.1798 0.1713>, rot=<0.0421 0.2586 0.0113 -0.965>}
 dyn_hairStrand_lower_J_01 {parent=dyn_hair_Mstr_l, loc=<4.6459 -2.8925 2.5863>, rot=<0.0477 0.084 -0.1146 -0.9887>}
 dyn_hairStrand_lower_J_02 {parent=dyn_hairStrand_lower_J_01, loc=<6.4044 0 0>, rot=<-0.0167 0.3823 0.0403 -0.923>}
 dyn_hairStrand_lower_K_01 {parent=dyn_hair_Mstr_l, loc=<2.3733 -0.4278 -0.7132>, rot=<0.0528 -0.3161 -0.0464 -0.9461>}
 dyn_hairStrand_lower_K_02 {parent=dyn_hairStrand_lower_K_01, loc=<6.3034 0 1.689>, rot=<0.0668 0.6403 0.056 -0.7631>}
 dyn_hairStrand_lower_L_01 {parent=dyn_hair_Mstr_l, loc=<1.8853 2.5428 -0.0679>, rot=<0.0468 0.0516 -0.1009 -0.9925>}
 dyn_hairStrand_lower_L_02 {parent=dyn_hairStrand_lower_L_01, loc=<4.7954 0 0>, rot=<0 0.342 0 -0.9397>}
 dyn_bang_A_01 {parent=head, loc=<6.7195 12.3635 2.4082>, rot=<-0.0368 -0.6953 -0.0439 -0.7165>}
 dyn_bang_A_02 {parent=dyn_bang_A_01, loc=<4.0424 -0.7128 0>, rot=<-0.0076 0.1735 0.2132 -0.9615>}
 dyn_bang_B_01 {parent=head, loc=<3.8454 10.2859 -0.8037>, rot=<0.1027 -0.7144 0.0337 -0.6913>}
 dyn_bang_B_02 {parent=dyn_bang_B_01, loc=<4.4249 0 0>, rot=<-0.0416 0.3004 -0.0131 -0.9528>}
 dyn_bang_C_01 {parent=head, loc=<5.5643 10.3618 5.5225>, rot=<-0.112 -0.9413 -0.0639 -0.3121>}
 dyn_bang_C_02 {parent=dyn_bang_C_01, loc=<4.6576 -0.4138 0.8213>, rot=<-0.1337 0.1693 0.251 -0.9436>}
 dyn_hairStrand_upper_E_01 {parent=head, loc=<3.9045 -10.8224 6.8786>, rot=<0.4176 -0.2639 0.8682 -0.0469>}
 dyn_hairStrand_upper_E_02 {parent=dyn_hairStrand_upper_E_01, loc=<5.2185 0 0.9202>, rot=<-0.0396 0.4175 -0.0182 -0.9076>}
 dyn_hairStrand_upper_E_03 {parent=dyn_hairStrand_upper_E_02, loc=<3.5872 0.1138 1.3007>, rot=<-0.0402 0.465 0.0369 -0.8836>}
 dyn_hairStrand_upper_F_01 {parent=head, loc=<4.6216 -12.6262 0.0231>, rot=<0.1911 -0.7749 0.537 -0.2731>}
 dyn_hairStrand_upper_F_02 {parent=dyn_hairStrand_upper_F_01, loc=<5.4102 0 0.954>, rot=<0.0812 0.5009 -0.0537 -0.86>}
 dyn_hairStrand_upper_G_01 {parent=head, loc=<6.4084 -11.496 -5.6273>, rot=<0.1217 -0.8758 0.329 -0.3315>}
 dyn_hairStrand_upper_G_02 {parent=dyn_hairStrand_upper_G_01, loc=<4.9044 0 2.287>, rot=<0.018 0.5575 -0.0047 -0.83>}
 dyn_hairStrand_upper_I_01 {parent=head, loc=<5.2446 -1.8196 -11.8319>, rot=<0.1586 -0.8874 0.1763 -0.3954>}
 dyn_hairStrand_upper_I_02 {parent=dyn_hairStrand_upper_I_01, loc=<3.7219 -0.3593 1.7356>, rot=<-0.0888 0.5713 0.0284 -0.8155>}
 dyn_hairStrand_lower_A_01 {parent=head, loc=<-0.5749 3.9765 8.552>, rot=<0.2413 0.2034 0.9463 0.0705>}
 dyn_hairStrand_lower_A_02 {parent=dyn_hairStrand_lower_A_01, loc=<4.1318 -0.0965 1.1029>, rot=<0.0178 0.5153 0.0106 -0.8568>}
 dyn_hairStrand_lower_A_03 {parent=dyn_hairStrand_lower_A_02, loc=<3.584 0.1314 2.1332>, rot=<0.0426 0.608 0.0717 -0.7896>}
 dyn_hairStrand_lower_M_01 {parent=head, loc=<-0.2136 1.4366 -9.1436>, rot=<0.0214 -0.9476 0.2734 -0.1641>}
 dyn_hairStrand_lower_M_02 {parent=dyn_hairStrand_lower_M_01, loc=<4.8459 0 0>, rot=<-0.0374 0.2268 0.0747 -0.9704>}
 dyn_hairStrand_lower_M_03 {parent=dyn_hairStrand_lower_M_02, loc=<3.8229 -0.3463 1.0244>, rot=<-0.0379 0.3581 0.0399 -0.9321>}
 dyn_hairStrand_lower_M_04 {parent=dyn_hairStrand_lower_M_03, loc=<3.6298 -0.3176 1.3262>, rot=<0.0595 0.4974 0.0595 -0.8634>}
 dyn_hairStrand_lower_N_01 {parent=head, loc=<1.5502 4.1792 -7.9116>, rot=<0.0377 -0.9768 -0.2064 -0.0421>}
 dyn_hairStrand_lower_N_02 {parent=dyn_hairStrand_lower_N_01, loc=<4.9556 0 0>, rot=<-0.0105 0.2805 -0.0407 -0.9589>}
 dyn_hairStrand_lower_N_03 {parent=dyn_hairStrand_lower_N_02, loc=<4.002 0 1.4566>, rot=<-0.0334 0.6422 -0.028 -0.7653>}
 faceAttach {parent=head, loc=<-4.146 -0.8545 0>, rot=<0.0029 -0.7071 -0.0029 -0.7071>}
 dyn_hairStrand_upper_A_01 {parent=head, loc=<3.8298 5.9769 8.5058>, rot=<0.6857 -0.1865 0.6878 -0.1485>}
 dyn_hairStrand_upper_A_02 {parent=dyn_hairStrand_upper_A_01, loc=<2.3621 0 0>, rot=<0.0321 0.2068 0.0518 -0.9765>}
 dyn_hairStrand_upper_B_01 {parent=head, loc=<3.5717 1.9816 11.0679>, rot=<0.7632 0.0105 0.6338 -0.1257>}
 dyn_hairStrand_upper_B_02 {parent=dyn_hairStrand_upper_B_01, loc=<2.3962 0 0.4225>, rot=<0.0756 0.6388 -0.0434 -0.7644>}
 dyn_hairStrand_upper_C_01 {parent=head, loc=<4.2199 -1.0172 11.6146>, rot=<0.5833 -0.103 0.7958 -0.1262>}
 dyn_hairStrand_upper_C_02 {parent=dyn_hairStrand_upper_C_01, loc=<2.7589 0 1.2865>, rot=<0.0858 0.5714 -0.0151 -0.816>}
 dyn_hairStrand_upper_J_01 {parent=head, loc=<4.0424 2.8182 -10.6585>, rot=<0.4114 -0.4071 0.071 -0.8124>}
 dyn_hairStrand_upper_J_02 {parent=dyn_hairStrand_upper_J_01, loc=<2.4468 -0.1697 0.6333>, rot=<0.0553 0.331 -0.0081 -0.942>}
 dyn_hairStrand_upper_K_01 {parent=head, loc=<7.3769 4.4316 -10.8663>, rot=<0.3676 -0.0768 0.1227 -0.9187>}
 dyn_hairStrand_upper_K_02 {parent=dyn_hairStrand_upper_K_01, loc=<1.9995 0 0>, rot=<0 0.1564 0 -0.9877>}
 clavicle_l {parent=spine_05, loc=<5.4831 -0.1053 -3.2946>, rot=<0.0465 -0.7289 -0.009 -0.6829>}
 upperarm_l {parent=clavicle_l, loc=<9.8658 -1.2136 -0.4105>, rot=<0.0164 -0.3245 -0.0269 -0.9454>}
 clavicle_r {parent=spine_05, loc=<5.4831 -0.1053 3.2946>, rot=<0.7289 0.0465 -0.6829 0.009>}
 upperarm_r {parent=clavicle_r, loc=<-9.8658 1.2136 0.4105>, rot=<0.0164 -0.3245 -0.0269 -0.9454>}
 attach_cape {parent=spine_05, loc=<9.017 -11.714 0>, rot=<0.0616 -0.7044 -0.0616 -0.7044>}
]
variables [
 Custom Emote Alpha {type=double, default=0}
 Is Skydiving {type=bool, default=False}
 Is Skydiving Dive Mode {type=bool, default=False}
 WaveSpeed 2 {type=double, default=2}
 WaveSpeed 3 {type=double, default=3}
 WaveSpeed 4 {type=double, default=4}
]
